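# Translucent UI material that combines the UI kit backplate styles into one, selected by the Backplate Switch parameter.
# Slate post-buffer blur feeds functions for styles A to D, with C darkened at low quality, while a fifth option is flat bright yellow.
# Opacity varies by style, blends toward opaque as UseBlur rises, and is multiplied by a rounded-corner mask with per-corner radii.
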Material /FortUILibrary/Systems/Backplate/Materials/M_UIKit_BackplateCombined
graph {
 properties {
  bUsedWithStaticMesh=false
  blendMode=BLEND_Translucent
  materialDomain=MD_UI
  pixelDepthOffsetMode=PDOM_Legacy
  usedShadingModels=Default Lit
 }
 nodes [
  MaterialExpressionConstant_1 = Constant() {r=0.95}
  MaterialExpressionConstant_0 = Constant() {desc="Loud button = no transparency", r=1}
  n1 = ScalarParameter() {parameterName=Backplate Switch, sliderMax=3}
  n2 = Switch!(
   SwitchValue=@n1
   Default=@MaterialExpressionConstant_1
   A=@MaterialExpressionConstant_1
   B=@MaterialExpressionConstant_1
   C=@MaterialExpressionConstant_1
   D=@MaterialExpressionConstant_1
   E=@MaterialExpressionConstant_0
  ) {
   constDefault=0.9
   inputs [
    {
     inputName=A
     input {
      expression=&self:MaterialExpressionConstant_1
      inputName=None
      mask=0
      maskA=0
      maskB=0
      maskG=0
      maskR=0
      outputIndex=0
     }
    }
    {
     inputName=B
     input {
      expression=&self:MaterialExpressionConstant_1
      inputName=None
      mask=0
      maskA=0
      maskB=0
      maskG=0
      maskR=0
      outputIndex=0
     }
    }
    {
     inputName=C
     input {
      expression=&self:MaterialExpressionConstant_1
      inputName=None
      mask=0
      maskA=0
      maskB=0
      maskG=0
      maskR=0
      outputIndex=0
     }
    }
    {
     inputName=D
     input {
      expression=&self:MaterialExpressionConstant_1
      inputName=None
      mask=0
      maskA=0
      maskB=0
      maskG=0
      maskR=0
      outputIndex=0
     }
    }
    {
     inputName=E
     input {
      expression=&self:MaterialExpressionConstant_0
      inputName=None
      mask=0
      maskA=0
      maskB=0
      maskG=0
      maskR=0
      outputIndex=0
     }
    }
   ]
  }
  n3 = VectorParameter() {
   group=Shape
   parameterName=Corner Radius
   sortPriority=1
   channelNames {a=BottomRight, b=TopRight, g=BottomLeft, r=TopLeft}
   defaultValue=<rgba 16 16 16 16>
  }
  Multiply(
   A=LinearInterpolate(A=@n2, Alpha=ScalarParameter() {defaultValue=1, parameterName=UseBlur})
   B=MaterialFunctionCall(
    Corner Radius=AppendVector(A=@n3, B=@n3)
    Radius in Pixels=ScalarParameter() {defaultValue=1, group=Shape, parameterName=Radius in Pixels, sliderMax=1, sortPriority=2}
   ) {materialFunction=&MF_UI_CoreUI_RoundedMask}.Shape Mask
  )
  NamedRerouteDeclaration(None=MaterialFunctionCall!("UVs (Optional)") {materialFunction=/Engine/Functions/UserInterface/GetSlatePost0}.LinearRGB) {name=Blur RGB, nodeColor=<rgba 0 0 0 1>}
  NamedRerouteDeclaration(None=MaterialFunctionCall!(Post Buffer Blur=NamedRerouteUsage()) {materialFunction=/FortUILibrary/MaterialFunctions/CoreUI/MF_UI_CoreUI_BackplateA}.Result) {name=Backplate A, nodeColor=<rgba 0 0 0 1>}
  NamedRerouteDeclaration(None=MaterialFunctionCall!(Post Buffer Blur=NamedRerouteUsage()) {materialFunction=/FortUILibrary/MaterialFunctions/CoreUI/MF_UI_CoreUI_BackplateB}.Result) {name=Backplate B, nodeColor=<rgba 0 0 0 1>}
  n4 = MaterialFunctionCall!(Post Buffer Blur=NamedRerouteUsage()) {materialFunction=/FortUILibrary/MaterialFunctions/CoreUI/MF_UI_CoreUI_BackplateC}
  NamedRerouteDeclaration(None=QualitySwitch(Default=@n4.Result, Low=Multiply(A=@n4.Result) {constB=0.5})) {name=Backplate C, nodeColor=<rgba 0 0 0 1>}
  NamedRerouteDeclaration(None=MaterialFunctionCall!(Post Buffer Blur=NamedRerouteUsage()) {materialFunction=/FortUILibrary/MaterialFunctions/CoreUI/MF_UI_CoreUI_BackplateD}.Result) {name=Backplate D, nodeColor=<rgba 0 0 0 1>}
  MaterialExpressionNamedRerouteUsage_2 = NamedRerouteUsage()
  MaterialExpressionNamedRerouteUsage_6 = NamedRerouteUsage()
  MaterialExpressionNamedRerouteUsage_7 = NamedRerouteUsage()
  MaterialExpressionNamedRerouteUsage_8 = NamedRerouteUsage()
  MaterialExpressionConstant3Vector_0 = Constant3Vector() {constant=<rgba 0.9301 1 0.0103 1>}
  n5 = Switch(
   SwitchValue=@n1
   A=@MaterialExpressionNamedRerouteUsage_2
   B=@MaterialExpressionNamedRerouteUsage_6
   C=@MaterialExpressionNamedRerouteUsage_7
   D=@MaterialExpressionNamedRerouteUsage_8
   E=@MaterialExpressionConstant3Vector_0
  ) {inputs [
    {
     inputName=A
     input {
      expression=&self:MaterialExpressionNamedRerouteUsage_2
      inputName=None
      mask=0
      maskA=0
      maskB=0
      maskG=0
      maskR=0
      outputIndex=0
     }
    }
    {
     inputName=B
     input {
      expression=&self:MaterialExpressionNamedRerouteUsage_6
      inputName=None
      mask=0
      maskA=0
      maskB=0
      maskG=0
      maskR=0
      outputIndex=0
     }
    }
    {
     inputName=C
     input {
      expression=&self:MaterialExpressionNamedRerouteUsage_7
      inputName=None
      mask=0
      maskA=0
      maskB=0
      maskG=0
      maskR=0
      outputIndex=0
     }
    }
    {
     inputName=D
     input {
      expression=&self:MaterialExpressionNamedRerouteUsage_8
      inputName=None
      mask=0
      maskA=0
      maskB=0
      maskG=0
      maskR=0
      outputIndex=0
     }
    }
    {
     inputName=E
     input {
      expression=&self:MaterialExpressionConstant3Vector_0
      inputName=None
      mask=1
      maskA=0
      maskB=1
      maskG=1
      maskR=1
      outputIndex=0
     }
    }
   ]}
  Reroute(None=@n5)
 ]
}
