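# UI material for a daily bonus goal reward icon in the quest screen, drawn translucent from a texture parameter.
# An IsCompleted parameter blends the icon color toward CoreUI BackplateD, fed by the Slate post buffer, and its alpha toward half strength.
# A separate BackplateC branch reads the same buffer and halves its result at low quality through a quality switch.
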Material /QuestUI/QuestScreen/Materials/M_UI_DailyBonusGoalReward
graph {
 properties {bAutomaticallySetUsageInEditor=false, blendMode=BLEND_Translucent, materialDomain=MD_UI, usedShadingModels=Default Lit}
 nodes [
  Constant4Vector() {constant=<rgba 0.0513 0.0513 0.0513 1>}
  n1 = TextureSample(Tex=TextureObjectParameter() {parameterName=Texture, texture=/FortUILibrary/Systems/Items/Textures/T_UI_SoldierCommando_F})
  n2 = MaterialFunctionCall!("UVs (Optional)") {materialFunction=/Engine/Functions/UserInterface/GetSlatePost0}
  LinearInterpolate(
   A=@n1.RGB
   B=MaterialFunctionCall!(Post Buffer Blur=@n2.LinearRGB) {materialFunction=/FortUILibrary/MaterialFunctions/CoreUI/MF_UI_CoreUI_BackplateD}.Result
   Alpha=ScalarParameter() {parameterName=IsCompleted}
  )
  LinearInterpolate(A=@n1.A, B=Multiply(A=@n1.A) {constB=0.5}, Alpha=ScalarParameter() {parameterName=IsCompleted})
  n3 = MaterialFunctionCall!(Post Buffer Blur=@n2.LinearRGB) {materialFunction=/FortUILibrary/MaterialFunctions/CoreUI/MF_UI_CoreUI_BackplateC}
  QualitySwitch(Default=@n3.Result, Low=Multiply(A=@n3.Result) {constB=0.5})
 ]
}
