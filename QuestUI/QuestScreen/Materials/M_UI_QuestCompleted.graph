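# Translucent UI material for a quest completion indicator, drawing a rounded shape with a checkmark icon.
# The check comes from a center-scaled MSDF texture, converted to an SDF and anti-aliased.
# That mask builds two color variants, green-and-dark and dark-and-white, named Completed and Available.
# IsRewardAvailable blends between the variants, and IsCompleted blends from a plain dark fill to that result.
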Material /QuestUI/QuestScreen/Materials/M_UI_QuestCompleted
graph {
 properties {bAutomaticallySetUsageInEditor=false, blendMode=BLEND_Translucent, materialDomain=MD_UI, usedShadingModels=Default Lit}
 nodes [
  LinearInterpolate(
   A=Constant4Vector() {constant=<rgba 0.0048 0.004 0.0116 1>}.RGB
   B=LinearInterpolate(A=NamedRerouteUsage(), B=NamedRerouteUsage(), Alpha=ScalarParameter() {defaultValue=1, parameterName=IsRewardAvailable})
   Alpha=ScalarParameter() {parameterName=IsCompleted}
  )
  MaterialFunctionCall(Corner Radius=Constant() {r=1}) {materialFunction=&MF_UI_CoreUI_RoundedMask}
  n1 = MaterialFunctionCall(Texture Scale=Constant2Vector() {g=0.7, r=0.7}) {materialFunction=&ScaleUVsByCenter}
  n2 = MaterialFunctionCall(
   DistanceField=MaterialFunctionCall!(MSDF RGB=TextureSample(UVs=@n1.UVs) {samplerType=SAMPLERTYPE_LinearColor, texture=/FortUILibrary/Textures/Icons/T_UI_Icon_UIKit_Check_MSDF}.RGB) {materialFunction=/FortUILibrary/MaterialFunctions/Shapes/Modifiers/MF_UI_MSDFtoSDF}.SDF
   Threshold=Constant() {r=0.5}
  ) {materialFunction=&MF_UI_AntiAliasing}
  NamedRerouteDeclaration(None=LinearInterpolate(
    A=Constant4Vector() {constant=<rgba 0.117 0.7682 0.1384 1>}.RGB
    B=Constant4Vector() {constant=<rgba 0.0048 0.004 0.0116 1>}.RGB
    Alpha=@n2.Outer
   )) {name=Completed, nodeColor=<rgba 0.4059 1 0.933 1>}
  NamedRerouteDeclaration(None=LinearInterpolate(A=Constant4Vector() {constant=<rgba 0.0048 0.004 0.0116 1>}.RGB, B=Constant4Vector() {constant=<rgba 1 1 1 1>}.RGB, Alpha=@n2.Outer)) {name=Available, nodeColor=<rgba 0.0717 0.0877 0.0885 1>}
 ]
}
