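# Translucent UI material for the quest screen's bonus goal reward background, with a solid or gradient base color.
# A radial glow, StormTech grid, bottom strip, noise-wobbled border glow and directional particle-opacity gradient form one highlight mask.
# That mask blends the base color and alpha toward the highlight color and opacity.
# A rounded-corner mask shapes the panel, and named reroutes carry the layers.
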Material /QuestUI/QuestScreen/Materials/M_UI_Quests_BonusGoalRewardBackgroundParticles
graph {
 properties {bAutomaticallySetUsageInEditor=false, blendMode=BLEND_Translucent, materialDomain=MD_UI, usedShadingModels=Default Lit}
 nodes [
  n1 = VectorParameter() {group=Background, parameterName=BackgroundColor}
  n2 = VectorParameter() {group=Background, parameterName=BackgroundColor2}
  n3 = Constant() {r=0.25}
  n4 = MaterialFunctionCall(
   UVs=NamedRerouteUsage()
   GradientStart=ScalarParameter() {group=Background, parameterName=BackgroundGradientStart}
   GradientEnd=ScalarParameter() {defaultValue=1, group=Background, parameterName=BackgroundGradientEnd}
   RotationAngle=@n3
  ) {materialFunction=&MF_UI_LinearGradient}
  n5 = VectorParameter() {group=Background, parameterName=BackgroundColor}
  n6 = Add(
   A=NamedRerouteUsage()
   B=Multiply(A=NamedRerouteUsage(), B=ScalarParameter() {defaultValue=1, group=Grid, parameterName=GridOpacity, sliderMax=1})
  )
  n7 = LinearInterpolate(
   A=@n6
   B=Add(A=@n6, B=NamedRerouteUsage())
   Alpha=ScalarParameter() {defaultValue=0.5, group=BottomStrip, parameterName=BottomStripOpacity, sliderMax=1}
  )
  n8 = Multiply(A=NamedRerouteUsage(), B=ScalarParameter() {defaultValue=1, group=Border, parameterName=BorderGlowOpacity, sliderMax=1})
  n9 = Power(
   Base=MaterialFunctionCall(
    UVs=NamedRerouteUsage()
    GradientStart=ScalarParameter() {group=Particles, parameterName=ParticlesOpacityEnd}
    RotationAngle=Constant() {r=0.75}
   ) {materialFunction=&MF_UI_LinearGradient}.Result
   Exp=ScalarParameter() {defaultValue=1, group=Particles, parameterName=ParticlesOpacityFalloff}
  )
  n10 = Multiply(A=Add(A=Saturate(None=@n7), B=Max(A=@n8, B=@n9)), B=NamedRerouteUsage())
  LinearInterpolate(
   A=StaticSwitchParameter(True=LinearInterpolate(A=@n1.RGB, B=@n2.RGB, Alpha=@n4.Result), False=@n5.RGB) {group=Background, parameterName=UseGradient}
   B=VectorParameter() {group=Highlights, parameterName=HighlightsColor, defaultValue=<rgba 0.7682 0.4125 0.1779 1>}.RGB
   Alpha=@n10
  )
  n11 = LinearInterpolate(
   A=@n1.A
   B=@n2.A
   Alpha=MaterialFunctionCall(
    UVs=NamedRerouteUsage()
    GradientStart=ScalarParameter() {group=Background, parameterName=BackgroundGradientStart}
    GradientEnd=ScalarParameter() {defaultValue=1, group=Background, parameterName=BackgroundGradientEnd}
    RotationAngle=@n3
   ) {materialFunction=&MF_UI_LinearGradient}.Result
  )
  LinearInterpolate(
   A=Multiply(A=StaticSwitchParameter(True=@n11, False=@n5.A) {group=Background, parameterName=UseGradient}, B=NamedRerouteUsage())
   B=ScalarParameter() {defaultValue=1, group=Highlights, parameterName=HighlightsOpacity, sliderMax=1}
   Alpha=@n10
  )
  NamedRerouteDeclaration(None=MaterialFunctionCall() {materialFunction=&GetUserInterfaceUV}.Normalized UV) {name=UV, nodeColor=<rgba 0.5234 1 0 1>}
  n12 = Abs(None=ConstantBiasScale(None=NamedRerouteUsage()) {bias=-0.5, scale=2})
  n13 = TextureSample(
   UVs=Panner(Coordinate=NamedRerouteUsage()) {speedY=0.1}
   Tex=TextureObject() {samplerType=SAMPLERTYPE_Grayscale, texture=&LowResBlurredNoise}
  )
  MaterialExpressionAdd_14 = Add(
   A=Multiply(A=@n12, B=@n12)
   B=Multiply(A=AppendVector(A=@n13.R, B=@n13.R), B=ScalarParameter() {defaultValue=0.15, group=Border, parameterName=BorderGlowWobble}) {constB=0.1}
  )
  n14 = Convert(None=@MaterialExpressionAdd_14) {
   nodeName=Break Float2
   convertInputs [{
     type=Vector2
     expressionInput {
      expression=&self:MaterialExpressionAdd_14
      inputName=None
      mask=0
      maskA=0
      maskB=0
      maskG=0
      maskR=0
      outputIndex=0
     }
    }]
   convertOutputs [{type=Scalar}, {type=Scalar}]
  }
  NamedRerouteDeclaration(None=Power(Base=LinearInterpolate(A=@n14, Alpha=@n14), Exp=ScalarParameter() {defaultValue=8, group=Border, parameterName=BorderGlowFalloff}) {constExponent=8}) {name=BorderGlow, nodeColor=<rgba 0 0.9062 1 1>}
  NamedRerouteDeclaration(None=Constant2Vector() {g=1, r=0.5}) {name=BottomPosition, nodeColor=<rgba 0.125 1 0 1>}
  n15 = VectorParameter() {
   group=Border
   parameterName=CornerRadius
   channelNames {a=Bottom Right, b=Top Right, g=Bottom Left, r=Top Left}
   defaultValue=<rgba 12 12 12 12>
  }
  NamedRerouteDeclaration(None=MaterialFunctionCall(UV=TextureCoordinate(), Corner Radius=@n15.RGBA, Radius in Pixels=Constant() {r=1}) {materialFunction=&MF_UI_CoreUI_RoundedMask}.Shape Mask) {name=BackgroundCorners, nodeColor=<rgba 0 1 0.2734 1>}
  NamedRerouteDeclaration(None=MaterialFunctionCall(UV=NamedRerouteUsage(), ScalePivotPoint=NamedRerouteUsage()) {materialFunction=&MF_UI_UVScaling}.ScaleX) {name=BottomUV, nodeColor=<rgba 0.0312 0 1 1>}
  n16 = Constant() {r=-0.5}
  n17 = Multiply(A=ScalarParameter() {defaultValue=0.1, group=BottomStrip, parameterName=BottomStripHeight}) {constB=0.5}
  n18 = Add(A=NamedRerouteUsage(), B=AppendVector(A=Constant(), B=Add(A=@n16, B=@n17)))
  n19 = AppendVector(A=Constant() {r=1}, B=ScalarParameter() {defaultValue=0.1, group=BottomStrip, parameterName=BottomStripHeight})
  n20 = NamedRerouteUsage()
  NamedRerouteDeclaration(None=Multiply(
    A=MaterialFunctionCall(UVs=@n18, Size=@n19) {materialFunction=&MF_UI_Box}.Fill
    B=MaterialFunctionCall!(UVs=@n20, Offset, RotationCenter, RotationAngle) {materialFunction=/FortUILibrary/MaterialFunctions/Gradients/MF_UI_ReflectedGradient}.Result
   )) {name=BottomStrip, nodeColor=<rgba 0.4297 0 1 1>}
  NamedRerouteDeclaration(None=AppendVector(
    A=ScalarParameter() {defaultValue=0.5, group=Glow, parameterName=GlowPositionX}
    B=ScalarParameter() {defaultValue=1, group=Glow, parameterName=GlowPositionY}
   )) {name=GlowPosition, nodeColor=<rgba 1 0.1641 0 1>}
  NamedRerouteDeclaration(None=MaterialFunctionCall!(
    Screen UV
    Tiling=ScalarParameter() {defaultValue=48, group=Grid, parameterName=GridTiling}
    Rotation
    Main Min/Max Opacity
    Main Min/Max Scale=Constant() {r=0.9}
    Main Min/Max Perspective=Constant()
    Main Outline Thickness=Constant()
    Main Speed
    Main Reduction
    Sub Min/Max Opacity
    Sub Min/Max Scale=Constant() {r=0.7}
    Sub Min/Max Perspective=Constant()
    Sub Outline Thickness=Constant()
    Sub Speed
    Sub Reduction
    Random Seed
    Time
   ) {materialFunction=/FortUILibrary/MaterialFunctions/BR/MF_UI_BR_StormTech}.Grid) {name=StormGrid, nodeColor=<rgba 1 0 0.9844 1>}
  n21 = MaterialFunctionCall(
   UVs=NamedRerouteUsage()
   Texture Scale=AppendVector(
    A=ScalarParameter() {defaultValue=1, group=Glow, parameterName=GlowWidth}
    B=ScalarParameter() {defaultValue=1, group=Glow, parameterName=GlowHeight}
   )
  ) {materialFunction=&ScaleUVsByCenter}
  n22 = NamedRerouteUsage()
  n23 = ScalarParameter() {defaultValue=0.5, group=Glow, parameterName=GlowRadius}
  n24 = ScalarParameter() {defaultValue=1, group=Glow, parameterName=GlowDensity}
  NamedRerouteDeclaration(None=Multiply(
    A=MaterialFunctionCall(UVs=@n21.UVs, CenterPosition=@n22, Radius=@n23, Density=@n24) {materialFunction=&RadialGradientExponential}.RadialGradientExponential
    B=ScalarParameter() {defaultValue=1, group=Glow, parameterName=GlowOpacity, sliderMax=1}
   )) {name=Glow, nodeColor=<rgba 0 0.6719 1 1>}
 ]
}
