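# Translucent UI material for the Morning Bell quest final reward, showing drifting particles within a glow.
# Two particle-function layers, foreground and background, differ in column count, panning, size variation and seed, with each grid squared to shape it.
# The glow sums two exponential radial gradients centered on the bottom edge, one scaled into a streak.
# Particles, glow and a linear edge gradient form a saturated mask, alongside a tint color parameter.
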Material /QuestUI/QuestScreen/Materials/M_UI_Quests_MorningBell_FinalReward
graph {
 properties {bAutomaticallySetUsageInEditor=false, blendMode=BLEND_Translucent, materialDomain=MD_UI, usedShadingModels=Default Lit}
 nodes [
  n1 = MaterialFunctionCall!(
   UV=NamedRerouteUsage()
   Columns=ScalarParameter() {defaultValue=5, group=Foreground, parameterName=Foreground Columns}
   Horizontal Panning Range
   Pan Speed=VectorParameter() {group=Foreground, parameterName=Foreground Panning Speed, defaultValue=<rgba 0 0.5 0 0>}.RGB
   Sizing Variation Speed=VectorParameter() {group=Foreground, parameterName=Foreground Sizing Variation, defaultValue=<rgba 0.1 0.2 0 0>}.RGB
   Sizing Variation Range
   Seed=ScalarParameter() {defaultValue=0.3, group=Foreground, parameterName=Foreground Seed}
  ) {materialFunction=&/MF_UI_Quests_MorningBell_Particles}
  NamedRerouteDeclaration(None=Multiply(A=@n1.Grid, B=@n1.Grid)) {name=Foreground Particles, nodeColor=<rgba 1 0.8203 0 1>}
  n2 = MaterialFunctionCall(
   UV=NamedRerouteUsage()
   Columns=ScalarParameter() {defaultValue=4, group=Background, parameterName=Background Columns}
   Horizontal Panning Range=Constant2Vector() {g=0.15, r=-0.15}
   Pan Speed=VectorParameter() {group=Background, parameterName=Background Panning Speed, defaultValue=<rgba 0 0.3 0 0>}.RGB
   Sizing Variation Range=VectorParameter() {group=Background, parameterName=Background Sizing Variation, defaultValue=<rgba 0.1 0.2 0 0>}.RGB
   Seed=ScalarParameter() {defaultValue=0.1, group=Background, parameterName=Background Seed}
  ) {materialFunction=&MF_UI_Quests_MorningBell_Particles}
  NamedRerouteDeclaration(None=Multiply(A=@n2.Grid, B=@n2.Grid)) {name=Background Particles, nodeColor=<rgba 0.5 0 1 1>}
  n3 = TextureCoordinate()
  NamedRerouteDeclaration(None=MaterialFunctionCall(UV=NamedRerouteDeclaration(None=@n3) {name=UV, nodeColor=<rgba 0.7109 0 1 1>}, ScalePivotPoint=Constant2Vector() {g=1, r=0.5}) {materialFunction=&MF_UI_UVScaling}.ScaleX) {name=Scaled UV, nodeColor=<rgba 0.9922 0 1 1>}
  n4 = NamedRerouteUsage()
  n5 = VectorParameter() {group=Glow, parameterName=Glow Streak Scle, defaultValue=<rgba 0.5 1 0 1>}
  n6 = Add(
   A=MaterialFunctionCall(
    UVs=MaterialFunctionCall(UVs=@n4, Texture Scale=@n5.RGB) {materialFunction=&ScaleUVsByCenter}.UVs
    CenterPosition=Constant2Vector() {g=1, r=0.5}
    Radius=ScalarParameter() {defaultValue=1, group=Glow, parameterName=Glow Base Radius}
    Density=ScalarParameter() {defaultValue=0.5, group=Glow, parameterName=Glow Density}
   ) {materialFunction=&RadialGradientExponential}.RadialGradientExponential
   B=MaterialFunctionCall(
    UVs=NamedRerouteUsage()
    CenterPosition=Constant2Vector() {g=1, r=0.5}
    Radius=ScalarParameter() {defaultValue=0.7, group=Glow, parameterName=Glow Streak Radius}
    Density=ScalarParameter() {defaultValue=0.5, group=Glow, parameterName=Glow Density}
   ) {materialFunction=&RadialGradientExponential}.RadialGradientExponential
  )
  NamedRerouteDeclaration(None=Multiply(A=Saturate(None=@n6), B=ScalarParameter() {defaultValue=5, group=Glow, parameterName=Glow Opacity, sliderMax=1})) {name=Glow, nodeColor=<rgba 0 0.3906 1 1>}
  n7 = NamedRerouteUsage()
  n8 = NamedRerouteUsage()
  n9 = Saturate(None=Multiply(A=Add(A=@n7, B=@n8)) {constB=10})
  n10 = NamedRerouteUsage()
  n11 = Constant()
  n12 = Constant() {r=0.2}
  n13 = Constant() {r=0.25}
  Saturate(None=Multiply(
    A=Multiply(A=@n9, B=@n10)
    B=MaterialFunctionCall(GradientStart=@n11, GradientEnd=@n12, RotationAngle=@n13) {materialFunction=&MF_UI_LinearGradient}.Result
   ))
  VectorParameter() {group=Particles, parameterName=Color, defaultValue=<rgba 0.6724 0.341 0.0967 1>}
 ]
}
