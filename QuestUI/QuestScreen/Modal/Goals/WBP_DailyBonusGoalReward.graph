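# Widget that shows one daily bonus goal reward in the quest screen: an icon, a completed mark and a quantity label.
# UpdateReward takes the first reward view model, async-loads its preview into the icon's dynamic material and collapses the icon on failure.
# SetIsCompleted compares the quest view model's achieved and required counts to drive the material's IsCompleted parameter and the mark.
# The label shows only quantities above one, abbreviating thousands as truncated "k" values.
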WidgetBlueprint /QuestUI/QuestScreen/Modal/Goals/WBP_DailyBonusGoalReward
graphs [
 EventGraph {nodes [
   n1 = UpdateIcon!() -> (OutputDelegate, then, LazyObject:"Soft Object Reference (by ref)") {is=title K2Node_CustomEvent}
   > n2 = Branch(Condition=IsValidSoftObjectReference(SoftObjectReference=@n1.LazyObject))
   > n3 = Async Load Asset(Asset=@n1.LazyObject)
   n4 = Cast To Texture!(execute=@n3.Completed, Object=@n3) -> (then, CastFailed, AsTexture:Texture Object Reference) {is=title K2Node_DynamicCast}
   > n5 = GetDynamicMaterial(self=Get LazyIcon() -> (LazyIcon:Fort Lazy Image Object Reference))
   > SetTextureParameterValue(self=@n5, ParameterName=Texture, Value=@n4)
   > SetVisibility(self=Get LazyIcon(), InVisibility=SelfHitTestInvisible)
   SetVisibility(execute=[@n4.CastFailed, @n2.else], self=Get LazyIcon(), InVisibility=Collapsed)
  ]}
 SetIsCompleted {nodes [
   SetIsCompleted() {is=title K2Node_FunctionEntry}
   > n1 = Get!(execute:Exec) -> (then:Exec, else:Exec, QuestUIQuestVM:Quest UI Quest View Model Object Reference)
   > n2 = GetDynamicMaterial(self=Get LazyIcon())
   n3 = "integer >= integer"(
    A=Get AchievedCount!(self:Quest UI Quest View Model Object Reference=@n1) -> (AchievedCount:int=0)
    B=Get RequiredCount!(self:Quest UI Quest View Model Object Reference=@n1) -> (RequiredCount:int=0)
   )
   > SetScalarParameterValue(self=@n2, ParameterName=IsCompleted, Value="To Float (Boolean)"(InBool=@n3))
   > SetVisibility(
    self=Get Image_Completed() -> (Image_Completed:Image Object Reference)
    InVisibility=Select(Option 0=Collapsed, Option 1=SelfHitTestInvisible, Index=@n3)
   )
  ]}
 UpdateReward {nodes [
   n1 = UpdateReward!() -> (then, Array:Array of Quest UI Reward VM Object References) {is=title K2Node_FunctionEntry}
   n2 = "Get (a copy)"(Array:Array of Quest UI Reward VM Object References=@n1) -> (Output:Quest UI Reward VM Object Reference)
   > Is Valid(InputObject=@n2)
   > UpdateIcon!(
    execute
    self
    LazyObject:"Soft Object Reference (by ref)"=Get SmallPreviewImage!(self:Quest UI Reward VM Object Reference=@n2) -> (SmallPreviewImage:Soft Object Reference)
   ) -> (then) {is=title K2Node_CallFunction}
   > SetIsCompleted() {is=title K2Node_CallFunction}
   n3 = Get Quantity!(self:Quest UI Reward VM Object Reference=@n2) -> (Quantity:int=0)
   n4 = "integer > integer"(A=@n3, B=1)
   > SetVisibility(
    self=Get Reward_Text() -> (Reward_Text:Common Text Object Reference)
    InVisibility=Select(Option 0=Hidden, Option 1=SelfHitTestInvisible, Index=@n4)
   )
   > Branch(Condition=@n4)
   n5 = int / int(A=@n3, B=1000)
   > "SetText (Text)"(
    self=Get Reward_Text()
    InText=Select(
     Option 0:Text="To Text (Integer)"(Value=@n3)
     Option 1:Text=Format Text!(Format="LOCTEXT(\"{Amount}k\")", Amount:int=@n5) -> (Result)
     Index="integer >= integer"(A=@n5, B=1)
    ) -> (ReturnValue:Text)
   )
  ]}
]
variables []
widgets {
 info {inheritedWidgetCount=0, parentClass=&CommonUserWidget, rootWidgetClass=&GridPanel, widgetCount=4}
 widgets [
  GridPanel_93 {widgetClassPath=&GridPanel, widget=GridPanel_93, properties {columnFill [1], rowFill [1]}}
  LazyIcon {
   widgetClassPath=&FortLazyImage
   bIsVariable=true
   parent=GridPanel_93
   widget=LazyIcon
   properties {
    brush {
     bIsDynamicallyLoaded=false
     drawAs=Image
     imageType=FullColor
     mirroring=NoMirror
     resourceName=None
     resourceObject=/QuestUI/QuestScreen/Materials/M_UI_DailyBonusGoalReward
     tiling=NoTile
     imageSize=<56 56>
     margin=<ltrb 0 0 0 0>
     outlineSettings {
      bUseBrushTransparency=false
      roundingType=HalfHeightRadius
      width=0
      color {colorUseRule=UseColor_Specified, specifiedColor=<rgba 0 0 0 0>}
      cornerRadii=<0 0 0 0>
     }
     tintColor {colorUseRule=UseColor_Specified, specifiedColor=<rgba 1 1 1 1>}
     uVRegion {bIsValid=false, max=<0 0>, min=<0 0>}
    }
    loadingBackgroundBrush {
     bIsDynamicallyLoaded=false
     drawAs=Image
     imageType=FullColor
     mirroring=NoMirror
     resourceName=None
     resourceObject=null
     tiling=NoTile
     imageSize=<56 56>
     margin=<ltrb 0 0 0 0>
     outlineSettings {
      bUseBrushTransparency=false
      roundingType=HalfHeightRadius
      width=0
      color {colorUseRule=UseColor_Specified, specifiedColor=<rgba 0 0 0 0>}
      cornerRadii=<0 0 0 0>
     }
     tintColor {colorUseRule=UseColor_Specified, specifiedColor=<rgba 1 1 1 1>}
     uVRegion {bIsValid=false, max=<0 0>, min=<0 0>}
    }
   }
   slot {class=GridSlot, properties {horizontalAlignment=HAlign_Center, verticalAlignment=VAlign_Top, padding=<ltrb 13 0 13 0>}}
  }
  Image_Completed {
   widgetClassPath=&Image
   bIsVariable=true
   parent=GridPanel_93
   widget=Image_Completed
   properties {
    visibility=Collapsed
    brush {
     bIsDynamicallyLoaded=false
     drawAs=Image
     imageType=FullColor
     mirroring=NoMirror
     resourceName=None
     resourceObject=/QuestUI/QuestScreen/MaterialInstances/MI_UI_QuestIsCompleted
     tiling=NoTile
     imageSize=<24 24>
     margin=<ltrb 0 0 0 0>
     outlineSettings {
      bUseBrushTransparency=false
      roundingType=HalfHeightRadius
      width=0
      color {colorUseRule=UseColor_Specified, specifiedColor=<rgba 0 0 0 0>}
      cornerRadii=<0 0 0 0>
     }
     tintColor {colorUseRule=UseColor_Specified, specifiedColor=<rgba 1 1 1 1>}
     uVRegion {bIsValid=false, max=<0 0>, min=<0 0>}
    }
   }
   slot {class=GridSlot, properties {horizontalAlignment=HAlign_Center, layer=1, verticalAlignment=VAlign_Center, padding=<ltrb 0 0 0 2>}}
  }
  Reward_Text {
   widgetClassPath=&CommonTextBlock
   bIsVariable=true
   parent=GridPanel_93
   widget=Reward_Text
   properties {
    text=25k
    font {
     bForceMonospaced=false
     bMaterialIsStencil=false
     fontMaterial=null
     fontObject=&HeadingNow
     letterSpacing=-7
     monospacedWidth=1
     size=15
     skewAmount=0
     typefaceFontName=75 Medium
     outlineSettings {
      bApplyOutlineToDropShadows=false
      bMiteredCorners=false
      bSeparateFillAlpha=false
      outlineMaterial=null
      outlineSize=1
      outlineColor=<rgba 0 0 0 1>
     }
    }
    shadowOffset=<0 0>
   }
   slot {class=GridSlot, properties {horizontalAlignment=HAlign_Center, row=1, verticalAlignment=VAlign_Top, padding=<ltrb 0 -8 0 0>}}
  }
 ]
}
