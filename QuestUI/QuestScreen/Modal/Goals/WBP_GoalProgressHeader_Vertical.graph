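# Vertical goal header list entry for the quest screen modal, showing a punch card name and expiration countdown over a tile background.
# PreConstruct sets cut and indicator-direction parameters on the background's dynamic material.
# OnListItemObjectSet unwraps the item and stores it as QuestUIGoalHeaderVM.
# Setter functions collapse the widget for an empty name, toggle countdown visibility, start the countdown and dim disabled content through render opacity.
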WidgetBlueprint /QuestUI/QuestScreen/Modal/Goals/WBP_GoalProgressHeader_Vertical
graphs [
 EventGraph {nodes [
   Event PreConstruct()
   > n1 = GetDynamicMaterial(self=Get Block_Background() -> (Block_Background:Image Object Reference))
   > SetScalarParameterValue(self=@n1, ParameterName=IsCut, Value=1)
   > SetScalarParameterValue(self=@n1, ParameterName=IndicatorDirection, Value=180)
   n2 = Event OnListItemObjectSet()
   > n3 = Cast To QuestUIGoalHeaderVM!(execute, Object=UnwrapItem!(Item:Object Reference=@n2.ListItemObject) -> (ReturnValue:Object Reference)) -> (then, CastFailed, AsQuest UIGoal Header VM:Quest UIGoal Header VM Object Reference) {is=title K2Node_DynamicCast}
   > Set QuestUIGoalHeaderVM(QuestUIGoalHeaderVM:Quest UIGoal Header VM Object Reference=@n3) -> (Output_Get:Quest UIGoal Header VM Object Reference)
   Event Construct()
   Event Tick()
  ]}
 HasExpirationTime {nodes [
   n1 = HasExpirationTime!() -> (then, HasExpirationTime:bool=~) {is=title K2Node_FunctionEntry}
   > SetVisibility(
    self=Get ExpirationTime() -> (ExpirationTime:Athena Challenge Date Time Text Block Object Reference)
    InVisibility=Select(Option 0=Collapsed, Option 1=HitTestInvisible, Index=@n1)
   )
  ]}
 SetContentEnabled {nodes [
   n1 = SetContentEnabled!() -> (then, bEnabled:bool=~) {is=title K2Node_FunctionEntry}
   > SetRenderOpacity(InOpacity=Select(Option 0:float=0.5, Option 1:float=1, Index=@n1) -> (ReturnValue:float))
  ]}
 SetExpirationTime {nodes [
   n1 = SetExpirationTime!() -> (then, ExpirationTime:Date Time Structure) {is=title K2Node_FunctionEntry}
   > SetDateTimeValue!(
    execute
    self:Common Date Time Text Block Object Reference=Get ExpirationTime()
    InDateTime:Date Time Structure=@n1
    bShowAsCountdown:bool=true
    InRefreshDelay:float=1
   ) -> (then)
  ]}
 SetPunchCardName {nodes [
   n1 = SetPunchCardName!() -> (then, InText:Text) {is=title K2Node_FunctionEntry}
   > SetVisibility(InVisibility=Select(Option 0:ESlateVisibility Enum=SelfHitTestInvisible, Option 1:ESlateVisibility Enum=Collapsed, Index=TextIsEmpty(InText=@n1)) -> (ReturnValue:ESlateVisibility Enum))
   > "SetText (Text)"(self=Get PunchCardName() -> (PunchCardName:Common Text Object Reference), InText=@n1)
  ]}
]
variables []
widgets {
 info {inheritedWidgetCount=0, parentClass=&CommonUserWidget, rootWidgetClass=&GridPanel, widgetCount=4}
 widgets [
  GridPanel_27 {widgetClassPath=&GridPanel, widget=GridPanel_27, properties {columnFill [1], rowFill [1]}}
  Block_Background {
   widgetClassPath=&Image
   bIsVariable=true
   parent=GridPanel_27
   widget=Block_Background
   properties {brush {
     bIsDynamicallyLoaded=false
     drawAs=Image
     imageType=FullColor
     mirroring=NoMirror
     resourceName=None
     resourceObject=/FortniteCoreUI/UI/Quests/MaterialInstances/MI_UI_Quest_TileBackground_States
     tiling=NoTile
     imageSize=<32 32>
     margin=<ltrb 0 0 0 0>
     outlineSettings {
      bUseBrushTransparency=false
      roundingType=HalfHeightRadius
      width=0
      color {colorUseRule=UseColor_Specified, specifiedColor=<rgba 0 0 0 0>}
      cornerRadii=<0 0 0 0>
     }
     tintColor {colorUseRule=UseColor_Specified, specifiedColor=<rgba 1 1 1 1>}
     uVRegion {bIsValid=false, max=<0 0>, min=<0 0>}
    }}
   slot {class=GridSlot, properties {columnSpan=2, rowSpan=2, padding=<ltrb 1 0 0 0>}}
  }
  PunchCardName {
   widgetClassPath=&CommonTextBlock
   bIsVariable=true
   parent=GridPanel_27
   widget=PunchCardName
   properties {
    autoWrapText=true
    text=Fortnite Mobile App Bonus Goals
    textTransformPolicy=ToUpper
    font {
     bForceMonospaced=false
     bMaterialIsStencil=false
     fontMaterial=null
     fontObject=&HeadingNow
     letterSpacing=0
     monospacedWidth=1
     size=20.25
     skewAmount=0
     typefaceFontName=76 Bold
     outlineSettings {
      bApplyOutlineToDropShadows=false
      bMiteredCorners=false
      bSeparateFillAlpha=false
      outlineMaterial=null
      outlineSize=0
      outlineColor=<rgba 0 0 0 1>
     }
    }
    shadowOffset=<0 0>
   }
   slot {class=GridSlot, properties {verticalAlignment=VAlign_Top, padding=<ltrb 24 14 24 0>}}
  }
  ExpirationTime {
   widgetClassPath=/Script/FortniteUI.AthenaChallengeDateTimeTextBlock
   bIsVariable=true
   parent=GridPanel_27
   widget=ExpirationTime
   properties {
    bAllowSeconds=false
    completionText=EXPIRED
    labelFormat="Ends: {0}"
    lineHeightPercentage=1.15
    text=Text Block
    colorAndOpacity {colorUseRule=UseColor_Specified, specifiedColor=<rgba 0.491 0.4793 0.5972 1>}
    font {
     bForceMonospaced=false
     bMaterialIsStencil=false
     fontMaterial=null
     fontObject=&HeadingNow
     letterSpacing=0
     monospacedWidth=1
     size=14.25
     skewAmount=0
     typefaceFontName=74 Regular
     outlineSettings {
      bApplyOutlineToDropShadows=false
      bMiteredCorners=false
      bSeparateFillAlpha=false
      outlineMaterial=null
      outlineSize=0
      outlineColor=<rgba 0 0 0 1>
     }
    }
    shadowOffset=<0 0>
   }
   slot {class=GridSlot, properties {horizontalAlignment=HAlign_Left, row=1, verticalAlignment=VAlign_Center, padding=<ltrb 24 0 0 4>}}
  }
 ]
}
