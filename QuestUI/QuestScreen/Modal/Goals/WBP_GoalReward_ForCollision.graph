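# Quest goal reward tile, built on WBP_QuestTile_Base, wrapping a WBP_QuestReward with outline, lock and particle layers.
# SetQuestVM stores the reward view model and shows the lock when its quest is missing.
# Punch card settings supply claimed, active or unclaimed backgrounds, async-loaded as material or texture, falling back to the default backplate.
# UpdateState restyles backgrounds by completion; selection drives outline and modular blocks, and gamepad input skips parent hover handling.
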WidgetBlueprint /QuestUI/QuestScreen/Modal/Goals/WBP_GoalReward_ForCollision
animations [ScrollAnimation {
  binding_count=0
  length_seconds=1
  object=&self:ScrollAnimation
  sequence {bindings []}
 }]
graphs [
 EventGraph {nodes [
   n1 = SetCustomStyles!() -> (OutputDelegate, then, CustomStyle:Soft Object Reference) {is=title K2Node_CustomEvent}
   n2 = "To Float (Boolean)"(InBool=IsValidSoftObjectReference(SoftObjectReference=@n1.CustomStyle))
   > SetRenderOpacity(self=Get Image_CustomBackground() -> (Image_CustomBackground:Image Object Reference), InOpacity=@n2)
   > SetRenderOpacity(self=Get Background() -> (Background:Image Object Reference), InOpacity=float - float(A=1, B=@n2))
   > n3 = Async Load Asset(Asset=@n1.CustomStyle)
   n4 = Cast To MaterialInstanceConstant(execute=@n3.Completed, Object=@n3)
   n5 = Get Image_CustomBackground()
   > SetBrushFromMaterial(self=@n5, Material=@n4)
   > n6 = GetDynamicMaterial(self=@n5)
   > Set MID_CustomBackground(MID_CustomBackground:Material Instance Dynamic Object Reference=@n6) -> (Output_Get:Material Instance Dynamic Object Reference)
   n7 = Cast To Texture2D(execute=@n4.CastFailed, Object=@n3)
   > SetBrushFromTexture!(execute, self:Image Object Reference=Get Image_CustomBackground(), Texture:Texture 2D Object Reference=@n7, bMatchSize:bool=false) -> (then)
   ResetRewardBackground(execute=@n7.CastFailed)
   Event Construct()
   > RegisterModularBlocks()
   Event PreConstruct()
   > n8 = GetDynamicMaterial(self=Get Background())
   > n9 = Set MID_Background(MID_Background:Material Instance Dynamic Object Reference=@n8) -> (Output_Get:Material Instance Dynamic Object Reference)
   > SetScalarParameterValue(self=@n9, ParameterName=Backplate Switch, Value=1)
   > SetVectorParameterValue!(execute, self=@n9, ParameterName=Corner Radius, Value_R:float=10, Value_G:float=10, Value_B:float=10, Value_A:float=10) -> (then)
   > SetQuestRewardStyle()
   Event On Selected()
   > SelectBlocks!(execute, self, IsSelected:bool=true) -> (then)
   n10 = Get Image_Outline() -> (Image_Outline:Image Object Reference)
   > SetVisibility(self=@n10, InVisibility=SelfHitTestInvisible)
   Event On Deselected()
   > SelectBlocks(IsSelected=false)
   > SetVisibility(self=@n10, InVisibility=Collapsed)
   Event On Hovered()
   n11 = Get CommonInputSubsystem()
   > n12 = Branch(Condition="Equal (Enum)"(A=GetCurrentInputType(self=@n11), B=Gamepad))
   Parent: On Hovered(execute=@n12.else) {is=title K2Node_CallParentFunction}
  ]}
 GetCustomBackground {nodes [
   GetCustomBackground!() -> (then, Completed:bool=~, Active:bool=~) {is=title K2Node_FunctionEntry}
   > n1 = Get!(
    execute:Exec
    self:Quest UIGoal Reward VM Object Reference=Get QuestUIGoalRewardVM() -> (QuestUIGoalRewardVM:Quest UIGoal Reward VM Object Reference)
   ) -> (then:Exec, else:Exec, QuestVM:Quest UI Quest View Model Object Reference)
   n2 = Break Fort Quest Goal Card Stage Custom Background!(FortQuestGoalCardStageCustomBackground:"Fort Quest Goal Card Stage Custom Background Structure (by ref)"=Break Fort Quest Goal Card Settings!(FortQuestGoalCardSettings:"Fort Quest Goal Card Settings Structure (by ref)"=Get PunchCardSettings!(self:Quest UI Quest View Model Object Reference=@n1) -> (PunchCardSettings:Fort Quest Goal Card Settings Structure)) -> (CustomGoalWidgetBackground:Fort Quest Goal Card Stage Custom Background Structure, CustomGoalIcon:Texture 2D Soft Object Reference) {is=title K2Node_BreakStruct}) -> (
    CustomClaimedStageGoalWidgetBackground:Soft Object Reference
    CustomActiveStageGoalWidgetBackground:Soft Object Reference
    CustomUnclaimedStageGoalWidgetBackground:Soft Object Reference
   ) {is=title K2Node_BreakStruct}
   n3 = Get Active() -> (Active:bool)
   > Return Node!(
    execute
    CustomBackground:Soft Object Reference=Select(
     Option 0:Soft Object Reference=Select(Option 0=@n2.CustomUnclaimedStageGoalWidgetBackground, Option 1=@n2.CustomActiveStageGoalWidgetBackground, Index=@n3)
     Option 1:Soft Object Reference=@n2
     Index=Get Completed() -> (Completed:bool)
    ) -> (ReturnValue:Soft Object Reference)
   )
  ]}
 OnInnerPartClicked {nodes [OnInnerPartClicked() {is=title K2Node_FunctionEntry}]}
 RegisterModularBlocks {nodes [
   RegisterModularBlocks() {is=title K2Node_FunctionEntry}
   > RegisterModularBlock!(
    execute
    self:UIKit Modular Button Object Reference
    Block:UIKit Block Object Reference=Get Block_Outline() -> (Block_Outline:WBP UIKit Block Outline Object Reference)
   ) -> (then)
  ]}
 ResetRewardBackground {nodes [
   ResetRewardBackground() {is=title K2Node_FunctionEntry}
   > SetRenderOpacity(self=Get Image_CustomBackground(), InOpacity=0)
   > SetRenderOpacity(self=Get Background(), InOpacity=1)
  ]}
 SelectBlocks {nodes [
   n1 = SelectBlocks!() -> (then, IsSelected:bool=~) {is=title K2Node_FunctionEntry}
   > n2 = For Each Loop(Array:Array of UIKit Block Object References=Get ModularBlocks() -> (ModularBlocks:Array of UIKit Block Object References)) -> (Array Element:UIKit Block Object Reference)
   > n3 = Cast To WBP_UIKit_Block_Base!(execute, Object=@n2) -> (then, CastFailed, AsWBP UIKit Block Base:WBP UIKit Block Base Object Reference) {is=title K2Node_DynamicCast}
   > SetIsSelected!(execute, self:WBP UIKit Block Base Object Reference=@n3, IsSelected:bool=@n1) -> (then)
  ]}
 SetDebugQuestID {nodes [
   n1 = SetDebugQuestID!() -> (then, QuestID:String) {is=title K2Node_FunctionEntry}
   > "SetText (Text)"(self=Get DebugQuestID() -> (DebugQuestID:Common Text Object Reference), InText="To Text (String)"(InString=@n1))
  ]}
 SetParticlesEffect {nodes [
   n1 = SetParticlesEffect!() -> (then, HasParticleEfect:bool=~) {is=title K2Node_FunctionEntry}
   > SetBrushFromMaterial(
    self=Get Background()
    Material=Select(
     Option 0:Material Interface Object Reference=/FortUILibrary/Systems/Backplate/Materials/M_UIKit_BackplateCombined
     Option 1:Material Interface Object Reference=/QuestUI/QuestScreen/Materials/M_UI_Quests_BonusGoalRewardBackgroundParticles
     Index=@n1
    ) -> (ReturnValue:Material Interface Object Reference)
   )
   > SetRenderOpacity(self=Get Particles() -> (Particles:Image Object Reference), InOpacity="To Float (Boolean)"(InBool=@n1))
   > n2 = Branch(Condition=@n1)
   UpdateState!(
    execute=@n2.else
    self
    Progress:double=Get ProgressValue!(self:Quest UIGoal Reward VM Object Reference=Get QuestUIGoalRewardVM()) -> (ProgressValue:double)
   ) -> (then)
   > n3 = GetDynamicMaterial(self=Get Background())
   > SetScalarParameterValue(self=@n3, ParameterName=Backplate Switch, Value=1)
   > SetVectorParameterValue(self=@n3, ParameterName=Corner Radius, Value_R=10, Value_G=10, Value_B=10, Value_A=10)
  ]}
 SetQuestRewardStyle {nodes [
   SetQuestRewardStyle() {is=title K2Node_FunctionEntry}
   n1 = Get QuestReward() -> (QuestReward:WBP Quest Reward Object Reference)
   > SetPadding!(
    execute
    self:Vertical Box Slot Object Reference=SlotAsVerticalBoxSlot(Widget=Get IconRoot!(self:WBP Quest Reward Object Reference=@n1) -> (IconRoot:Overlay Object Reference))
    InPadding_Left:float=4
    InPadding_Top:float=4
    InPadding_Right:float=4
    InPadding_Bottom:float=4
   ) -> (then)
   > SetPadding!(
    execute
    self=SlotAsVerticalBoxSlot(Widget=Get TextRoot!(self:WBP Quest Reward Object Reference=@n1) -> (TextRoot:Overlay Object Reference))
    InPadding:Margin Structure
   ) -> (then)
   n2 = Get QuestReward()
   > UpdateRewardRadius!(execute, self:WBP Quest Reward Object Reference=@n2, InPixels:bool=true, CornerRadius:double=8) -> (then)
   n3 = Get Pages!(self:WBP Quest Reward Object Reference=@n2) -> (Pages:WBP Quest Pages Object Reference)
   > SetStyle!(execute, self:WBP Quest Pages Object Reference=@n3, RegularStyle:bool=false) -> (then)
   > AddChild(self=Get CommonVisualAttachment_0() -> (CommonVisualAttachment_0:Common Visual Attachment Object Reference), Content=@n3)
  ]}
 SetQuestVM {nodes [
   n1 = SetQuestVM!() -> (then, QuestVM:Quest UIGoal Reward VM Object Reference) {is=title K2Node_FunctionEntry}
   > ResetRewardBackground() {is=title K2Node_CallFunction}
   n2 = Get QuestVM!(self:Quest UIGoal Reward VM Object Reference=@n1) -> (QuestVM:Quest UI Quest View Model Object Reference)
   > SetVisibility(
    self=Get Image_Lock() -> (Image_Lock:Image Object Reference)
    InVisibility=Select(Option 0:ESlateVisibility Enum=SelfHitTestInvisible, Option 1:ESlateVisibility Enum=Collapsed, Index=IsValid(Object=@n2)) -> (ReturnValue:ESlateVisibility Enum)
   )
   > Set QuestUIGoalRewardVM(QuestUIGoalRewardVM:Quest UIGoal Reward VM Object Reference=@n1) -> (Output_Get:Quest UIGoal Reward VM Object Reference)
   n3 = Get QuestReward()
   > SetSelectedPage!(execute, self:WBP Quest Pages Object Reference=Get Pages(self=@n3), SelectedOption:int=0) -> (then)
   > Set CurrentIndex!(execute, CurrentIndex:int=0, self:WBP Quest Reward Object Reference=@n3) -> (then, Output_Get:int=0)
  ]}
 UpdateCustomBackground {nodes [
   UpdateCustomBackground() {is=title K2Node_FunctionEntry}
   n1 = Get QuestUIGoalRewardVM()
   n2 = Get ProgressValue(self=@n1)
   > SetCustomStyles!(
    execute
    self
    CustomStyle:Soft Object Reference=GetCustomBackground!(
     Completed:bool="float >= float"(A=@n2, B=1)
     Active:bool=Get NextReward!(self:Quest UIGoal Reward VM Object Reference=@n1) -> (NextReward:bool)
    ) -> (CustomBackground:Soft Object Reference) {is=title K2Node_CallFunction}
   ) -> (then) {is=title K2Node_CallFunction}
  ]}
 UpdateState {nodes [
   n1 = UpdateState!() -> (then, Progress:double=~) {is=title K2Node_FunctionEntry}
   n2 = "float >= float"(A=@n1, B=1)
   > SetScalarParameterValue(
    self=Get MID_CustomBackground() -> (MID_CustomBackground:Material Instance Dynamic Object Reference)
    ParameterName=Trans_SelectedIdle
    Value="To Float (Boolean)"(InBool=@n2)
   )
   > SetScalarParameterValue(
    self=Get MID_Background() -> (MID_Background:Material Instance Dynamic Object Reference)
    ParameterName=Backplate Switch
    Value=Select(Option 0:float=2, Option 1:float=1, Index=@n2) -> (ReturnValue:float)
   )
   > SetRenderOpacity(self=Get Overlay_Backgronds() -> (Overlay_Backgronds:Overlay Object Reference), InOpacity=Select(Option 0=0.5, Option 1=1, Index=@n2))
  ]}
]
variables [
 bGoal {type=boolean}
 BlockTiming {
  type=UIKitBlockTiming
  default {
   deselectingFocused=0
   deselectingIdle=0.1
   disabling=0
   enabling=0
   focusing=0.1
   hovering=0.1
   pressing=0
   releasing=0
   selecting=0.1
   unfocusing=0.1
   unhovering=0.1
  }
 }
 MID_Background {type=MaterialInstanceDynamic}
 MID_CustomBackground {type=MaterialInstanceDynamic}
 OnInnerPartClicked {}
 ProgressBackgroundColor {type=LinearColor, default=<rgba 0.0048 0.004 0.0116 0.7>}
 ProgressColor {type=LinearColor, default=<rgba 0 0.8714 0.9823 1>}
 ProgressCompletedColor {type=LinearColor, default=<rgba 0.0331 0.9911 0.2195 1>}
 ProgressEmptyBackgroundColor {type=LinearColor, default=<rgba 0.7682 0.7682 0.7682 0.4>}
 QuestBackground {type=MaterialInstance, default=/FortniteCoreUI/UI/Quests/MaterialInstances/MI_UI_QuestPunchCardReward_Background}
]
widgets {
 info {
  inheritedWidgetCount=1
  parentClass=/QuestUI/QuestScreen/QuestTile/WBP_QuestTile_Base.WBP_QuestTile_Base_C
  rootWidgetClass=&Overlay
  widgetCount=15
 }
 widgets [
  Overlay_0 {widgetClassPath=&Overlay, widget=Overlay_0}
  Overlay_1 {
   widgetClassPath=&Overlay
   parent=Overlay_0
   widget=Overlay_1
   slot {class=OverlaySlot, properties {horizontalAlignment=HAlign_Center, verticalAlignment=VAlign_Center}}
  }
  Overlay_Backgronds {
   widgetClassPath=&Overlay
   bIsVariable=true
   parent=Overlay_1
   widget=Overlay_Backgronds
   slot {class=OverlaySlot, properties {horizontalAlignment=HAlign_Fill, verticalAlignment=VAlign_Fill}}
  }
  Image_CustomBackground {
   widgetClassPath=&Image
   bIsVariable=true
   parent=Overlay_Backgronds
   widget=Image_CustomBackground
   properties {
    renderOpacity=0
    visibility=SelfHitTestInvisible
    brush {
     bIsDynamicallyLoaded=false
     drawAs=Image
     imageType=FullColor
     mirroring=NoMirror
     resourceName=None
     resourceObject=&MI_UI_QuestPunchCardReward_Background
     tiling=NoTile
     imageSize=<116 116>
     margin=<ltrb 0 0 0 0>
     outlineSettings {
      bUseBrushTransparency=false
      roundingType=HalfHeightRadius
      width=0
      color {colorUseRule=UseColor_Specified, specifiedColor=<rgba 0 0 0 0>}
      cornerRadii=<0 0 0 0>
     }
     tintColor {colorUseRule=UseColor_Specified, specifiedColor=<rgba 1 1 1 1>}
     uVRegion {bIsValid=false, max=<0 0>, min=<0 0>}
    }
   }
   slot {class=OverlaySlot}
  }
  Background {
   widgetClassPath=&Image
   bIsVariable=true
   parent=Overlay_Backgronds
   widget=Background
   properties {
    visibility=SelfHitTestInvisible
    brush {
     bIsDynamicallyLoaded=false
     drawAs=Image
     imageType=FullColor
     mirroring=NoMirror
     resourceName=None
     resourceObject=&M_UIKit_BackplateCombined
     tiling=NoTile
     imageSize=<116 116>
     margin=<ltrb 0 0 0 0>
     outlineSettings {
      bUseBrushTransparency=false
      roundingType=HalfHeightRadius
      width=0
      color {colorUseRule=UseColor_Specified, specifiedColor=<rgba 0 0 0 0>}
      cornerRadii=<0 0 0 0>
     }
     tintColor {colorUseRule=UseColor_Specified, specifiedColor=<rgba 1 1 1 1>}
     uVRegion {bIsValid=false, max=<0 0>, min=<0 0>}
    }
   }
   slot {class=OverlaySlot}
  }
  Image_Outline {
   widgetClassPath=&Image
   bIsVariable=true
   parent=Overlay_1
   widget=Image_Outline
   properties {
    visibility=Collapsed
    brush {
     bIsDynamicallyLoaded=false
     drawAs=Image
     imageType=FullColor
     mirroring=NoMirror
     resourceName=None
     resourceObject=/QuestUI/QuestScreen/QuestTile/Materials/MI_UI_Outline_QuestGoal_Selected
     tiling=NoTile
     imageSize=<32 32>
     margin=<ltrb 0 0 0 0>
     outlineSettings {
      bUseBrushTransparency=false
      roundingType=HalfHeightRadius
      width=0
      color {colorUseRule=UseColor_Specified, specifiedColor=<rgba 0 0 0 0>}
      cornerRadii=<0 0 0 0>
     }
     tintColor {colorUseRule=UseColor_Specified, specifiedColor=<rgba 1 1 1 1>}
     uVRegion {bIsValid=false, max=<0 0>, min=<0 0>}
    }
   }
   slot {class=OverlaySlot, properties {horizontalAlignment=HAlign_Fill, verticalAlignment=VAlign_Fill}}
  }
  Block_Outline {
   widgetClassPath=/FortUILibrary/UIKit/Buttons/Blocks/WBP_UIKit_Block_Outline
   bIsVariable=true
   parent=Overlay_1
   widget=Block_Outline
   properties {bottomLeftCornerRadius=20, bottomRightCornerRadius=20, radiusInPixels=true, topLeftCornerRadius=20, topRightCornerRadius=20}
   slot {class=OverlaySlot, properties {horizontalAlignment=HAlign_Fill, verticalAlignment=VAlign_Fill}}
  }
  CommonVisualAttachment_46 {
   widgetClassPath=&CommonVisualAttachment
   parent=Overlay_1
   widget=CommonVisualAttachment_46
   properties {contentAnchor=<0.5 0.4>}
   slot {class=OverlaySlot, properties {horizontalAlignment=HAlign_Center}}
  }
  Particles {
   widgetClassPath=&Image
   bIsVariable=true
   parent=CommonVisualAttachment_46
   widget=Particles
   properties {
    renderOpacity=0
    brush {
     bIsDynamicallyLoaded=false
     drawAs=Image
     imageType=FullColor
     mirroring=NoMirror
     resourceName=None
     resourceObject=/QuestUI/QuestScreen/Materials/M_UI_Quests_MorningBell_FinalReward
     tiling=NoTile
     imageSize=<128 65>
     margin=<ltrb 0 0 0 0>
     outlineSettings {
      bUseBrushTransparency=false
      roundingType=HalfHeightRadius
      width=0
      color {colorUseRule=UseColor_Specified, specifiedColor=<rgba 0 0 0 0>}
      cornerRadii=<0 0 0 0>
     }
     tintColor {colorUseRule=UseColor_Specified, specifiedColor=<rgba 1 1 1 1>}
     uVRegion {bIsValid=false, max=<0 0>, min=<0 0>}
    }
   }
   slot {class=SizeBoxSlot}
  }
  QuestReward {
   widgetClassPath=/QuestUI/QuestScreen/QuestTile/WBP_QuestReward
   bIsVariable=true
   parent=Overlay_0
   widget=QuestReward
   properties {bHideBackgroundOverride=true, rewardIconSize=<75 75>, widgetSize=<86 78>}
   slot {class=OverlaySlot, properties {horizontalAlignment=HAlign_Fill, verticalAlignment=VAlign_Fill}}
  }
  FortVisualAttachment_73 {
   widgetClassPath=&FortVisualAttachment
   parent=Overlay_0
   widget=FortVisualAttachment_73
   properties {bOverride_MaxDesiredWidth=true, maxDesiredWidth=147, contentAnchor=<0.5 1>}
   slot {class=OverlaySlot, properties {horizontalAlignment=HAlign_Center, verticalAlignment=VAlign_Fill}}
  }
  DebugQuestID {
   widgetClassPath=&CommonTextBlock
   bIsVariable=true
   parent=FortVisualAttachment_73
   widget=DebugQuestID
   properties {
    bAutoCollapseWithEmptyText=true
    clipping=ClipToBoundsWithoutIntersecting
    justification=Center
    pixelSnapping=SnapToPixel
    scrollStyle=&/TextScrollStyle_RewardDebug.TextScrollStyle_RewardDebug_C
    text=Debug Quest ID
    colorAndOpacity {colorUseRule=UseColor_Specified, specifiedColor=<rgba 0.5776 0.5776 0.5776 1>}
    font {
     bForceMonospaced=false
     bMaterialIsStencil=false
     fontMaterial=null
     fontObject=&Roboto
     letterSpacing=0
     monospacedWidth=1
     size=6.75
     skewAmount=0
     typefaceFontName=Bold
     outlineSettings {
      bApplyOutlineToDropShadows=false
      bMiteredCorners=false
      bSeparateFillAlpha=false
      outlineMaterial=null
      outlineSize=0
      outlineColor=<rgba 0 0 0 1>
     }
    }
    shadowOffset=<0 0>
   }
   slot {class=SizeBoxSlot, properties {horizontalAlignment=HAlign_Center}}
  }
  CommonVisualAttachment_0 {
   widgetClassPath=&CommonVisualAttachment
   bIsVariable=true
   parent=Overlay_0
   widget=CommonVisualAttachment_0
   properties {contentAnchor=<0.5 0.5>}
   slot {class=OverlaySlot, properties {horizontalAlignment=HAlign_Center, verticalAlignment=VAlign_Bottom}}
  }
  Image_Lock {
   widgetClassPath=&Image
   bIsVariable=true
   parent=Overlay_0
   widget=Image_Lock
   properties {
    visibility=Collapsed
    brush {
     bIsDynamicallyLoaded=false
     drawAs=Image
     imageType=FullColor
     mirroring=NoMirror
     resourceName=None
     resourceObject=/FortUILibrary/Textures/MaterialInstances/MI_UIKit_Icon_Locked
     tiling=NoTile
     imageSize=<32 32>
     margin=<ltrb 0 0 0 0>
     outlineSettings {
      bUseBrushTransparency=false
      roundingType=HalfHeightRadius
      width=0
      color {colorUseRule=UseColor_Specified, specifiedColor=<rgba 0 0 0 0>}
      cornerRadii=<0 0 0 0>
     }
     tintColor {colorUseRule=UseColor_Specified, specifiedColor=<rgba 1 1 1 1>}
     uVRegion {bIsValid=false, max=<0 0>, min=<0 0>}
    }
   }
   slot {class=OverlaySlot, properties {horizontalAlignment=HAlign_Center, verticalAlignment=VAlign_Center}}
  }
  InputActionWidget {widgetClassPath=&CommonActionWidget, bInherited=true, widget=null}
 ]
}
