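# WBP_GoalReward_Solo is a quest UI widget showing one goal reward with a progress bar and a required-count or icon badge.
# It wraps WBP_GoalReward_ForCollision, forwarding the view model to it and relaying its click and selection events.
# Progress drives dynamic material parameters and picks claimed, active or unclaimed custom backgrounds from goal card settings.
# Reward updates refresh the badge, marking completion only when progress is full and reward content is available.
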WidgetBlueprint /QuestUI/QuestScreen/Modal/Goals/WBP_GoalReward_Solo
graphs [
 EventGraph {nodes [
   n1 = "On Selected (WBP_GoalReward_Colision)"!() -> (OutputDelegate, then, Button:Common Button Base Object Reference) {is=title K2Node_ComponentBoundEvent}
   > Call On Selected!(execute:Exec, self:WBP Goal Reward Solo Object Reference, Button:Common Button Base Object Reference=@n1.Button) -> (then:Exec) {is=title K2Node_CallDelegate}
   n2 = "On Clicked (WBP_GoalReward_Colision)"!() -> (OutputDelegate, then, Button:Common Button Base Object Reference) {is=title K2Node_ComponentBoundEvent}
   > Call On Clicked!(execute:Exec, self:WBP Goal Reward Solo Object Reference, Button:Common Button Base Object Reference=@n2.Button) -> (then:Exec) {is=title K2Node_CallDelegate}
   n3 = SetQuestVM!() -> (OutputDelegate, then, QuestUIGoalRewardVM:Quest UIGoal Reward VM Object Reference) {is=title K2Node_CustomEvent}
   > n4 = Set QuestUIGoalRewardVM(QuestUIGoalRewardVM:Quest UIGoal Reward VM Object Reference=@n3.QuestUIGoalRewardVM) -> (Output_Get:Quest UIGoal Reward VM Object Reference)
   > SetQuestVM!(
    execute
    self:WBP Goal Reward For Collision Object Reference=Get WBP_GoalReward_Colision() -> (WBP_GoalReward_Colision:WBP Goal Reward For Collision Object Reference)
    QuestVM:Quest UIGoal Reward VM Object Reference=@n4
   ) -> (then) {is=title K2Node_CallFunction}
   n5 = HandleRewardUpdated!() -> (OutputDelegate, then, RewardVM:Quest UI Reward VM Object Reference) {is=title K2Node_CustomEvent}
   > UpdateStatusIndicator()
   Event Destruct()
   n6 = Get QuestReward!(self:WBP Goal Reward For Collision Object Reference=Get WBP_GoalReward_Colision()) -> (QuestReward:WBP Quest Reward Object Reference)
   > Unbind Event from On Reward Updated(self:WBP Quest Reward Object Reference=@n6, Delegate=@n5) {is=title K2Node_RemoveDelegate}
   Event Construct()
   > Bind Event to On Reward Updated(self:WBP Quest Reward Object Reference=@n6, Delegate=@n5) {is=title K2Node_AddDelegate}
  ]}
 GetCustomBackground {nodes [
   GetCustomBackground!() -> (then, Completed:bool=~, Active:bool=~) {is=title K2Node_FunctionEntry}
   > n1 = Get!(
    execute:Exec
    self:Quest UIGoal Reward VM Object Reference=Get QuestUIGoalRewardVM() -> (QuestUIGoalRewardVM:Quest UIGoal Reward VM Object Reference)
   ) -> (then:Exec, else:Exec, QuestVM:Quest UI Quest View Model Object Reference)
   n2 = Break Fort Quest Goal Card Stage Custom Background!(FortQuestGoalCardStageCustomBackground:"Fort Quest Goal Card Stage Custom Background Structure (by ref)"=Break Fort Quest Goal Card Settings!(FortQuestGoalCardSettings:"Fort Quest Goal Card Settings Structure (by ref)"=Get PunchCardSettings!(self:Quest UI Quest View Model Object Reference=@n1) -> (PunchCardSettings:Fort Quest Goal Card Settings Structure)) -> (CustomGoalWidgetBackground:Fort Quest Goal Card Stage Custom Background Structure, CustomGoalIcon:Texture 2D Soft Object Reference) {is=title K2Node_BreakStruct}) -> (
    CustomClaimedStageGoalWidgetBackground:Soft Object Reference
    CustomActiveStageGoalWidgetBackground:Soft Object Reference
    CustomUnclaimedStageGoalWidgetBackground:Soft Object Reference
   ) {is=title K2Node_BreakStruct}
   n3 = Get Active() -> (Active:bool)
   > Return Node!(
    execute
    CustomBackground:Soft Object Reference=Select(
     Option 0:Soft Object Reference=Select(Option 0=@n2.CustomUnclaimedStageGoalWidgetBackground, Option 1=@n2.CustomActiveStageGoalWidgetBackground, Index=@n3)
     Option 1:Soft Object Reference=@n2
     Index=Get Completed() -> (Completed:bool)
    ) -> (ReturnValue:Soft Object Reference)
   )
  ]}
 OnClicked {nodes [OnClicked!() -> (then, Button:Common Button Base Object Reference) {is=title K2Node_FunctionEntry}]}
 OnSelected {nodes [OnSelected!() -> (then, Button:Common Button Base Object Reference) {is=title K2Node_FunctionEntry}]}
 ResetRewardBackground {nodes [
   ResetRewardBackground() {is=title K2Node_FunctionEntry}
   > SetRenderOpacity(
    self=Get Image_CustomBackground!(self:WBP Goal Reward For Collision Object Reference=Get WBP_GoalReward_Colision()) -> (Image_CustomBackground:Image Object Reference)
    InOpacity=0
   )
   > SetRenderOpacity(
    self=Get Background!(self:WBP Goal Reward For Collision Object Reference=Get WBP_GoalReward_Colision()) -> (Background:Image Object Reference)
    InOpacity=1
   )
  ]}
 SelectBlocks {nodes [
   n1 = SelectBlocks!() -> (then, IsSelected:bool=~) {is=title K2Node_FunctionEntry}
   > n2 = For Each Loop(Array:Array of UIKit Block Object References=Get ModularBlocks!(self:UIKit Modular Button Object Reference=Get WBP_GoalReward_Colision()) -> (ModularBlocks:Array of UIKit Block Object References)) -> (Array Element:UIKit Block Object Reference)
   > n3 = Cast To WBP_UIKit_Block_Base!(execute, Object=@n2) -> (then, CastFailed, AsWBP UIKit Block Base:WBP UIKit Block Base Object Reference) {is=title K2Node_DynamicCast}
   > SetIsSelected!(execute, self:WBP UIKit Block Base Object Reference=@n3, IsSelected:bool=@n1) -> (then)
  ]}
 SetDebugQuestID {nodes [
   n1 = SetDebugQuestID!() -> (then, QuestID:String) {is=title K2Node_FunctionEntry}
   > "SetText (Text)"(
    self=Get DebugQuestID!(self:WBP Goal Reward For Collision Object Reference=Get WBP_GoalReward_Colision()) -> (DebugQuestID:Common Text Object Reference)
    InText="To Text (String)"(InString=@n1)
   )
  ]}
 SetIndex {nodes [
   SetIndex!() -> (then, Index:int) {is=title K2Node_FunctionEntry}
   > n1 = Get(self=Get QuestUIGoalRewardVM())
   > "SetText (Text)"(
    self=Get Text_Required() -> (Text_Required:Common Text Object Reference)
    InText="To Text (Integer)"(Value=Get RequiredCount!(self:Quest UI Quest View Model Object Reference=@n1) -> (RequiredCount:int=0))
   )
  ]}
 SetProgressValue {nodes [
   n1 = SetProgressValue!() -> (then, Value:double=~) {is=title K2Node_FunctionEntry}
   > UpdateCustomBackground()
   > UpdateProgressState!(execute, self, ProgressValue:float=@n1, ProgressWidget:Image Object Reference=Get Progress() -> (Progress:Image Object Reference)) -> (then)
   > UpdateProgressState(ProgressValue=@n1)
   > UpdateStatusIndicator()
   > Branch(Condition="float >= float"(A=Get Value() -> (Value:double), B=1))
   > Set with Broadcast NextReward!(execute, NextReward:bool=false, self:Quest UIGoal Reward VM Object Reference=Get QuestUIGoalRewardVM()) -> (then, Output_Get:bool)
  ]}
 SetRequiredIcon {nodes [
   n1 = SetRequiredIcon!() -> (then, Icon:Texture 2D Soft Object Reference) {is=title K2Node_FunctionEntry}
   > n2 = Branch(Condition=IsValidSoftObjectReference(SoftObjectReference=@n1))
   n3 = Get Switcher_Required() -> (Switcher_Required:Common Visibility Switcher Object Reference)
   > SetActiveWidgetIndex(self:Common Visibility Switcher Object Reference=@n3, Index=1)
   > SetBrushFromLazyTexture!(
    execute
    self:Common Lazy Image Object Reference=Get LazyImage_RequiredIcon() -> (LazyImage_RequiredIcon:Fort Lazy Image Object Reference)
    LazyTexture:"Texture 2D Soft Object Reference (by ref)"=@n1
    bMatchSize:bool=false
   ) -> (then)
   SetActiveWidgetIndex(execute=@n2.else, self=@n3, Index=0)
  ]}
 SetShowProgress {nodes [
   n1 = SetShowProgress!() -> (then, Show:bool=~) {is=title K2Node_FunctionEntry}
   > SetVisibility(
    self=Get Panel_Progress() -> (Panel_Progress:Grid Panel Object Reference)
    InVisibility=Select(Option 0:ESlateVisibility Enum=Collapsed, Option 1:ESlateVisibility Enum=SelfHitTestInvisible, Index=@n1) -> (ReturnValue:ESlateVisibility Enum)
   )
  ]}
 UpdateCustomBackground {nodes [
   UpdateCustomBackground() {is=title K2Node_FunctionEntry}
   n1 = Get QuestUIGoalRewardVM()
   n2 = Get ProgressValue!(self:Quest UIGoal Reward VM Object Reference=@n1) -> (ProgressValue:double)
   > SetCustomStyles!(
    execute
    self:WBP Goal Reward For Collision Object Reference=Get WBP_GoalReward_Colision()
    CustomStyle:Soft Object Reference=GetCustomBackground!(
     Completed:bool="float >= float"(A=@n2, B=1)
     Active:bool=Get NextReward!(self:Quest UIGoal Reward VM Object Reference=@n1) -> (NextReward:bool)
    ) -> (CustomBackground:Soft Object Reference) {is=title K2Node_CallFunction}
   ) -> (then)
  ]}
 UpdateProgressState {nodes [
   n1 = UpdateProgressState!() -> (then, ProgressValue:float=~, ProgressWidget:Image Object Reference) {is=title K2Node_FunctionEntry}
   > Is Valid(InputObject=@n1.ProgressWidget)
   > n2 = GetDynamicMaterial(self=@n1.ProgressWidget)
   > SetScalarParameterValue(self=@n2, ParameterName=AnimateProgress, Value=@n1)
   > n3 = GetDynamicMaterial(self=Get IndexBackground() -> (IndexBackground:Image Object Reference))
   n4 = "float >= float"(A=@n1, B=1)
   > SetScalarParameterValue(self=@n3, ParameterName=IsCompleted, Value="To Float (Boolean)"(InBool=@n4))
   > SetVisibility(self=Get Switcher_Required(), InVisibility=Select(Option 0=SelfHitTestInvisible, Option 1=Hidden, Index=@n4))
  ]}
 UpdateState {nodes [
   n1 = UpdateState!() -> (then, Progress:double=~) {is=title K2Node_FunctionEntry}
   > UpdateState!(execute, self:WBP Goal Reward For Collision Object Reference=Get WBP_GoalReward_Colision(), Progress:double=@n1) -> (then) {is=title K2Node_CallFunction}
  ]}
 UpdateStatusIndicator {nodes [
   UpdateStatusIndicator() {is=title K2Node_FunctionEntry}
   > n1 = Get!(execute, self:WBP Quest Reward Object Reference=Get QuestReward(self=Get WBP_GoalReward_Colision())) -> (then, else, QuestUIRewardVM:Quest UI Reward VM Object Reference)
   > n2 = GetDynamicMaterial(self=Get IndexBackground())
   n3 = Get QuestUIGoalRewardVM()
   n4 = AND Boolean(
    A="float >= float"(A=Get ProgressValue(self=@n3), B=1)
    B=Get bIsRewardContentAvailable!(self:Quest UI Reward VM Object Reference=@n1) -> (bIsRewardContentAvailable:bool)
   )
   > SetScalarParameterValue(self=@n2, ParameterName=IsCompleted, Value="To Float (Boolean)"(InBool=@n4))
   > SetVisibility(self=Get Switcher_Required(), InVisibility=Select(Option 0=SelfHitTestInvisible, Option 1=Hidden, Index=@n4))
  ]}
]
variables [
 BlockTiming {
  type=UIKitBlockTiming
  default {
   deselectingFocused=0
   deselectingIdle=0.1
   disabling=0
   enabling=0
   focusing=0.1
   hovering=0.1
   pressing=0
   releasing=0
   selecting=0.1
   unfocusing=0.1
   unhovering=0.1
  }
 }
 OnClicked {}
 OnSelected {}
]
widgets {
 info {inheritedWidgetCount=0, parentClass=&CommonUserWidget, rootWidgetClass=&GridPanel, widgetCount=11}
 widgets [
  GridPanel_28 {widgetClassPath=&GridPanel, widget=GridPanel_28, properties {columnFill [1]}}
  Panel_Progress {
   widgetClassPath=&GridPanel
   bIsVariable=true
   parent=GridPanel_28
   widget=Panel_Progress
   properties {columnFill [1], rowFill [1]}
   slot {class=GridSlot, properties {verticalAlignment=VAlign_Bottom, padding=<ltrb 0 16 35 12>}}
  }
  Overlay_0 {
   widgetClassPath=&Overlay
   parent=Panel_Progress
   widget=Overlay_0
   slot {class=GridSlot, properties {horizontalAlignment=HAlign_Right, verticalAlignment=VAlign_Center}}
  }
  IndexBackground {
   widgetClassPath=&Image
   bIsVariable=true
   parent=Overlay_0
   widget=IndexBackground
   properties {
    visibility=SelfHitTestInvisible
    brush {
     bIsDynamicallyLoaded=false
     drawAs=RoundedBox
     imageType=FullColor
     mirroring=NoMirror
     resourceName=None
     resourceObject=/QuestUI/QuestScreen/Materials/M_UI_QuestCompleted
     tiling=NoTile
     imageSize=<40 40>
     margin=<ltrb 0 0 0 0>
     outlineSettings {
      bUseBrushTransparency=false
      roundingType=HalfHeightRadius
      width=0
      color {colorUseRule=UseColor_Specified, specifiedColor=<rgba 0 0 0 0>}
      cornerRadii=<0 0 0 0>
     }
     tintColor {colorUseRule=UseColor_Specified, specifiedColor=<rgba 1 1 1 1>}
     uVRegion {bIsValid=false, max=<0 0>, min=<0 0>}
    }
   }
   slot {class=OverlaySlot}
  }
  Switcher_Required {
   widgetClassPath=&CommonVisibilitySwitcher
   bIsVariable=true
   parent=Overlay_0
   widget=Switcher_Required
   slot {class=OverlaySlot, properties {horizontalAlignment=HAlign_Center, verticalAlignment=VAlign_Center, padding=<ltrb 3 3 3 3>}}
  }
  SizeBox_0 {
   widgetClassPath=&SizeBox
   parent=Switcher_Required
   widget=SizeBox_0
   properties {bOverride_WidthOverride=true, widthOverride=35}
   slot {class=CommonVisibilitySwitcherSlot}
  }
  ScaleBox_1 {
   widgetClassPath=&ScaleBox
   parent=SizeBox_0
   widget=ScaleBox_1
   slot {class=SizeBoxSlot, properties {horizontalAlignment=HAlign_Center}}
  }
  Text_Required {
   widgetClassPath=&CommonTextBlock
   bIsVariable=true
   parent=ScaleBox_1
   widget=Text_Required
   properties {
    scrollStyle=&TextScrollStyle-Base_C
    text=100
    font {
     bForceMonospaced=false
     bMaterialIsStencil=false
     fontMaterial=null
     fontObject=&HeadingNow
     letterSpacing=0
     monospacedWidth=1
     size=13.5
     skewAmount=0
     typefaceFontName=75 Medium
     outlineSettings {
      bApplyOutlineToDropShadows=false
      bMiteredCorners=false
      bSeparateFillAlpha=false
      outlineMaterial=null
      outlineSize=0
      outlineColor=<rgba 0 0 0 1>
     }
    }
    shadowOffset=<0 0>
   }
   slot {class=ScaleBoxSlot}
  }
  LazyImage_RequiredIcon {
   widgetClassPath=&FortLazyImage
   bIsVariable=true
   parent=Switcher_Required
   widget=LazyImage_RequiredIcon
   properties {brush {
     bIsDynamicallyLoaded=false
     drawAs=Image
     imageType=NoImage
     mirroring=NoMirror
     resourceName=None
     resourceObject=null
     tiling=NoTile
     imageSize=<34 34>
     margin=<ltrb 0 0 0 0>
     outlineSettings {
      bUseBrushTransparency=false
      roundingType=HalfHeightRadius
      width=0
      color {colorUseRule=UseColor_Specified, specifiedColor=<rgba 0 0 0 0>}
      cornerRadii=<0 0 0 0>
     }
     tintColor {colorUseRule=UseColor_Specified, specifiedColor=<rgba 1 1 1 1>}
     uVRegion {bIsValid=false, max=<0 0>, min=<0 0>}
    }}
   slot {class=CommonVisibilitySwitcherSlot, properties {horizontalAlignment=HAlign_Center, verticalAlignment=VAlign_Center}}
  }
  Progress {
   widgetClassPath=&Image
   bIsVariable=true
   parent=Panel_Progress
   widget=Progress
   properties {
    visibility=SelfHitTestInvisible
    brush {
     bIsDynamicallyLoaded=false
     drawAs=Image
     imageType=FullColor
     mirroring=NoMirror
     resourceName=None
     resourceObject=/QuestUI/QuestScreen/MaterialInstances/MI_UI_QuestProgressBar
     tiling=NoTile
     imageSize=<85 10>
     margin=<ltrb 0 0 0 0>
     outlineSettings {
      bUseBrushTransparency=false
      roundingType=HalfHeightRadius
      width=0
      color {colorUseRule=UseColor_Specified, specifiedColor=<rgba 0 0 0 0>}
      cornerRadii=<0 0 0 0>
     }
     tintColor {colorUseRule=UseColor_Specified, specifiedColor=<rgba 1 1 1 1>}
     uVRegion {bIsValid=false, max=<0 0>, min=<0 0>}
    }
   }
   slot {class=GridSlot, properties {verticalAlignment=VAlign_Center, padding=<ltrb 0 0 37 0>}}
  }
  WBP_GoalReward_Colision {
   widgetClassPath=&/WBP_GoalReward_ForCollision
   bIsVariable=true
   parent=GridPanel_28
   widget=WBP_GoalReward_Colision
   properties {bIsFocusable=false, bShouldSelectUponReceivingFocus=false, visibility=HitTestInvisible}
   slot {class=GridSlot, properties {column=1, horizontalAlignment=HAlign_Center, layer=2}}
  }
 ]
}
