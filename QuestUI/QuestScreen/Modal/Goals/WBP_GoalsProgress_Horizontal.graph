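# Quest screen list-entry button that shows a horizontal punch card's goal rewards and their progress.
# PopulateListView builds goal view models and picks a solo reward, an entry box, or a scrolling list with a separate final reward.
# Refresh computes each goal's progress, subtracting earlier requirements on Horizontal-orientation cards, and finds the current quest.
# Selection handles arrows, clicks and gamepad focus, toggles bundle and item views for custom-widget cards, and broadcasts changes.
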WidgetBlueprint /QuestUI/QuestScreen/Modal/Goals/WBP_GoalsProgress_Horizontal
graphs [
 ChangeRewardView {nodes [
   n1 = ChangeRewardView!() -> (then, InItemView:bool=~, OnlyVisual:bool=~) {is=title K2Node_FunctionEntry}
   n2 = Get CurrentState() -> (CurrentState:E_QuestPunchCardState Enum=NewEnumerator0)
   n3 = Get InItemView() -> (InItemView:bool)
   n4 = Get Is Item View() -> (Is Item View:bool)
   > Branch(Condition=AND Boolean(
     A="Not Equal (Boolean)"(A=@n3, B=@n4) -> (ReturnValue=~)
     B="Not Equal (Enum)"!(A:E_QuestPunchCardState Enum=@n2, B:E_QuestPunchCardState Enum=NewEnumerator2) -> (ReturnValue:bool)
    ))
   > n5 = Branch(Condition=@n1.OnlyVisual)
   n6 = Set Is Item View(execute=[@n12.then, @n5.then], Is Item View:bool=@n1) -> (Output_Get:bool)
   > n7 = Branch(Condition=@n6)
   > n8 = Branch(Condition=IsGamepad!() -> (ReturnValue:bool))
   > n9 = ChangeSelectedIndex!(execute, self, Index:int=0) -> (then)
   UpdateOutlineVisibility(execute=[@n9.then, @n8.else])
   ClearAllSelection(execute=@n7.else)
   > UpdateFocus()
   > n10 = Branch(Condition=GetSelected!(self:Common Button Base Object Reference) -> (ReturnValue:bool))
   UpdateOutlineVisibility(execute=[@n11.then, @n10.then])
   n11 = RequestInstantTransition!(
    execute=@n10.else
    self:UIKit Block Object Reference=Get Block_Outline() -> (Block_Outline:WBP UIKit Block Outline Object Reference)
    State:EUIKitBlockInstantTransitionState Enum=InstantUnfocused
   ) -> (then)
   Call On Reward View Changed!(execute:Exec=@n5.else, self:WBP Goals Progress Horizontal Object Reference, InItemView:bool=@n1) -> (then:Exec) {is=title K2Node_CallDelegate}
   > n12 = SetBottomButtonVisibility!(
    execute
    self:WBP Bottom Bar Decoy Button Object Reference=Get DecoyButtonBack() -> (DecoyButtonBack:WBP Bottom Bar Decoy Button Object Reference)
    ShouldButtonBeVisible:bool=AND Boolean!(A=@n1, B=IsCustomWidget!() -> (HasCustom:bool), C:bool=GetSelected()) -> (ReturnValue)
   ) -> (then)
  ]}
 ChangeSelectedIndex {nodes [
   ChangeSelectedIndex!() -> (then, Index:int) {is=title K2Node_FunctionEntry}
   n1 = GetIsBundleView!() -> (ReturnValue:bool)
   n2 = "integer > integer"(A=GetNumEntries!(self:Dynamic Entry Box Base Object Reference=Get DynamicEntryBox_Fill() -> (DynamicEntryBox_Fill:Dynamic Entry Box Object Reference)) -> (ReturnValue:int=0))
   n3 = "integer > integer"(A=GetNumItems!(self:List View Object Reference=Get ListView() -> (ListView:Common List View Object Reference)) -> (ReturnValue:int=0))
   n4 = Get Index() -> (Index:int=0)
   > Branch(Condition=AND Boolean(A=OR Boolean(A=@n3, B=@n2), B="integer >= integer"(A=@n4), C=NOT Boolean(A=@n1)))
   > n5 = Set SelectedIndex(SelectedIndex:int=Get Index() -> (Index=0)) -> (Output_Get:int=0)
   > Call On Selected Index Changed!(execute:Exec, self:WBP Goals Progress Horizontal Object Reference, Index:int=@n5) -> (then:Exec) {is=title K2Node_CallDelegate}
   > n6 = Switch on E_QuestPunchCardState!(execute, Selection:E_QuestPunchCardState Enum=Get CurrentState() -> (CurrentState=NewEnumerator0)) -> (NewEnumerator0:Exec, NewEnumerator1:Exec, NewEnumerator2:Exec) {is=title K2Node_SwitchEnum}
   n7 = Get QuestUIGoalRewardVM!(self:WBP Goal Reward Object Reference=Get finalGoalReward() -> (finalGoalReward:WBP Goal Reward Object Reference)) -> (QuestUIGoalRewardVM:Quest UIGoal Reward VM Object Reference)
   n8 = Get Index() -> (Index=0)
   n9 = Get Index!(self:Quest UIGoal Reward VM Object Reference=@n7) -> (Index=0)
   > n10 = Branch(Condition=AND Boolean(A=IsFinalRewardVisible!() -> (ReturnValue:bool), B=IsValid(Object=@n7), C="Equal (Integer)"(A=@n8, B=@n9)))
   > SetIsSelected!(execute, self:Common Button Base Object Reference=Get finalGoalReward(), InSelected:bool=true, bGiveClickFeedback:bool=true) -> (then)
   > Clear Selection(self:List View Object Reference=Get ListView())
   ClearSelection(execute=@n10.else, self:Common Button Base Object Reference=Get finalGoalReward())
   n11 = Get ListView()
   > Branch(Condition="integer < integer"(A=Get Index!() -> (Index=0), B=GetNumItems(self=@n11) -> (ReturnValue=0)))
   > ClearSelection(self=Get finalGoalReward())
   n12 = Get ListView()
   n13 = Get Index() -> (Index=0)
   > n14 = Branch(Condition="Not Equal (Object)"(
     A:Object Reference=Get Selected Item!(self:List View Object Reference=@n12) -> (ReturnValue:Object Reference) {is=title K2Node_CallFunction}
     B:Object Reference=GetItemAt!(self:List View Object Reference=@n12, Index:int=@n13) -> (ReturnValue:Object Reference)
    ) -> (ReturnValue:bool=~) {is=title K2Node_PromotableOperator})
   n15 = Get Index() -> (Index=0)
   n16 = Get ListView()
   > SetSelectedIndex!(execute, self:List View Object Reference=@n16, Index:int=@n15) -> (then)
   > n17 = ScrollIndexIntoView!(execute, self:List View Object Reference=@n16, Index:int=@n15) -> (then)
   n18 = Branch(execute=[@n17.then, @n14.else], Condition=GetIsBundleView())
   > n19 = ChangeRewardView!(execute, self, InItemView:bool=true, OnlyVisual:bool=false) -> (then) {is=title K2Node_CallFunction}
   UpdateFocus(execute=[@n19.then, @n18.else, @n23.then])
   n20 = Get DynamicEntryBox_Fill()
   n21 = Cast To WBP_GoalReward_Container!(
    execute=@n6.NewEnumerator1
    Object="Get (a copy)"(
     Array:Array of User Widget Object References=GetAllEntries!(self:Dynamic Entry Box Base Object Reference=@n20) -> (ReturnValue:Array of User Widget Object References)
     Dimension 1=Get Index() -> (Index=0)
    ) -> (Output:User Widget Object Reference)
   ) -> (then, CastFailed, AsWBP Goal Reward Container:WBP Goal Reward Container Object Reference) {is=title K2Node_DynamicCast}
   n22 = Get WBP_GoalReward_Colision!(self:WBP Goal Reward Container Object Reference=@n21) -> (WBP_GoalReward_Colision:WBP Goal Reward For Collision Object Reference)
   > SetIsSelected(self=@n22, InSelected=true, bGiveClickFeedback=false)
   > n23 = Set CurrentSelectedItem(CurrentSelectedItem:Common Button Base Object Reference=@n22) -> (Output_Get:Common Button Base Object Reference)
  ]}
 ClearAllSelection {nodes [
   ClearAllSelection() {is=title K2Node_FunctionEntry}
   > n1 = Switch on E_QuestPunchCardState(Selection=Get CurrentState() -> (CurrentState=NewEnumerator0))
   > Clear Selection(self=Get ListView())
   > ClearSelection(self=Get finalGoalReward())
   n2 = Get!(execute:Exec=@n1.NewEnumerator1) -> (then:Exec, else:Exec, CurrentSelectedItem:Common Button Base Object Reference)
   > ClearSelection(self=@n2)
   > Set CurrentSelectedItem()
  ]}
 EventGraph {nodes [
   n1 = "On Item Selection Changed (ListView)"!() -> (OutputDelegate, then, Item:Object Reference, bIsSelected:bool) {is=title K2Node_ComponentBoundEvent}
   > Branch(Condition=@n1.bIsSelected)
   > n2 = Cast To QuestUIGoalRewardVM!(execute, Object=@n1.Item) -> (then, CastFailed, AsQuest UIGoal Reward VM:Quest UIGoal Reward VM Object Reference) {is=title K2Node_DynamicCast}
   > n3 = Set As Quest UIGoal Reward VM(As Quest UIGoal Reward VM:Quest UIGoal Reward VM Object Reference=@n2) -> (Output_Get:Quest UIGoal Reward VM Object Reference)
   ChangeSelectedIndex(
    execute=[@n24.then, @n3.then]
    Index=Get Index!(self:Quest UIGoal Reward VM Object Reference=Get As Quest UIGoal Reward VM() -> (As Quest UIGoal Reward VM:Quest UIGoal Reward VM Object Reference)) -> (Index=0)
   ) {is=title K2Node_CallFunction}
   n4 = "On Clicked (finalGoalReward)"!() -> (OutputDelegate, then, Button:Common Button Base Object Reference) {is=title K2Node_ComponentBoundEvent}
   > n5 = Cast To WBP_GoalReward!(execute, Object=@n4.Button) -> (then, CastFailed, AsWBP Goal Reward:WBP Goal Reward Object Reference) {is=title K2Node_DynamicCast}
   > SelectFinalGoalReward!(execute, self, Widget:Common Button Base Object Reference=@n5) -> (then)
   n6 = "On List View Scrolled (ListView)"!() -> (OutputDelegate, then, ItemOffset:float, DistanceRemaining:float) {is=title K2Node_ComponentBoundEvent}
   n7 = Ceil(A=@n6.ItemOffset) {is=title K2Node_CallFunction}
   n8 = Get QuestUIHorizontalPunchCardVM() -> (QuestUIHorizontalPunchCardVM:Quest UI Horizontal Punch Card View Model Object Reference)
   n9 = int - int(
    A=Length(TargetArray:Array of Quest UI Base Content View Model Object References=Get OrderedContentVMs!(self:Quest UI Punch Card View Model Object Reference=@n8) -> (OrderedContentVMs:Array of Quest UI Base Content View Model Object References))
    B=int + int(A=Get ItemCountScrollThreshold() -> (ItemCountScrollThreshold:int=0))
   )
   > SetProgressBarType!(
    execute
    self:WBP Goal Reward Object Reference=Get finalGoalReward()
    WithSeparator:bool=AND Boolean!(A=IsFinalGoalVisible!() -> (ReturnValue:bool), B="integer >= integer"(A=@n9, B=@n7)) -> (ReturnValue)
   ) -> (then)
   "On Clicked (GoalsHorizontal_PreviousButton)"!() -> (OutputDelegate, then, Button:Common Button Base Object Reference) {is=title K2Node_ComponentBoundEvent}
   > SelectAdjacentEntry!(execute, self, Right:bool=false) -> (then)
   "On Clicked (GoalsHorizontal_NextButton)"!() -> (OutputDelegate, then, Button:Common Button Base Object Reference) {is=title K2Node_ComponentBoundEvent}
   > SelectAdjacentEntry(Right=true)
   n10 = "On Selected Changed (finalGoalReward)"!() -> (OutputDelegate, then, Button:Common Button Base Object Reference, Selected:bool) {is=title K2Node_ComponentBoundEvent}
   > Branch(Condition=@n10.Selected)
   > n11 = SelectFinalGoalReward(Widget=@n10.Button)
   Branch(execute=[@n11.then, @n14.then, @n18.then, @n15.then], Condition=NOT Boolean(A=IsGamepad()))
   > n12 = Branch(Condition=IsCustomWidget())
   n13 = GetIsBundleView()
   > Branch(Condition=@n13)
   > ChangeRewardView(InItemView=true, OnlyVisual=false)
   SetIsSelected(execute=@n12.else, InSelected=true, bGiveClickFeedback=false)
   n14 = "On Item Clicked (ListView)"!() -> (OutputDelegate, then, Item:Object Reference) {is=title K2Node_ComponentBoundEvent}
   n15 = "On Selected (WBP_GoalReward_Solo)"!() -> (OutputDelegate, then, Button:Common Button Base Object Reference) {is=title K2Node_ComponentBoundEvent}
   n16 = "On Pressed (FocusCatcher)"() {is=title K2Node_ComponentBoundEvent}
   Branch(execute=[@n16.then, @n17.then, @n26.then], Condition=@n13)
   > ChangeRewardView(InItemView=true, OnlyVisual=false)
   "On Clicked (DecoyButtonBack)"!() -> (OutputDelegate, then, Button:Common Button Base Object Reference) {is=title K2Node_ComponentBoundEvent}
   > ChangeRewardView(InItemView=false, OnlyVisual=false)
   n17 = "On Clicked (FocusCatcher)"() {is=title K2Node_ComponentBoundEvent}
   n18 = OnDynamicBoxEntryClicked!() -> (OutputDelegate, then, Button:Common Button Base Object Reference) {is=title K2Node_CustomEvent}
   n19 = OnDynamicBoxEntrySelected!() -> (OutputDelegate, then, Button:Common Button Base Object Reference) {is=title K2Node_CustomEvent}
   > n20 = Get()
   > n21 = ClearSelection(self=@n20)
   n22 = Set CurrentSelectedItem(execute=[@n21.then, @n20.else], CurrentSelectedItem=@n19.Button)
   > n23 = Cast To WBP_GoalReward_ForCollision!(execute, Object=@n22) -> (then, CastFailed, AsWBP Goal Reward for Collision:WBP Goal Reward For Collision Object Reference) {is=title K2Node_DynamicCast}
   > n24 = Set As Quest UIGoal Reward VM(As Quest UIGoal Reward VM=Get QuestUIGoalRewardVM(self:WBP Goal Reward For Collision Object Reference=@n23))
   Event On Entry Released() {is=title K2Node_Event}
   > InstantDeselect()
   Event Construct()
   > Parent: Construct()
   > SetDebugBundleID!(execute, self, BundleID:"String (by ref)"=MakeLiteralString()) -> (then)
   > Bind Event to On Input Method Changed(self=Get CommonInputSubsystem(), Delegate=Create Event())
   > SetBottomButtonVisibility(self=Get DecoyButtonBack(), ShouldButtonBeVisible=false)
   Event On Hovered()
   > n25 = Parent: On Hovered() {is=title K2Node_CallParentFunction}
   UpdateFocus(execute=[@n25.then, @n37.then])
   Event On Clicked()
   > n26 = Branch(Condition=IsGamepad())
   Branch(execute=@n26.else, Condition=GetIsItemView!() -> (ReturnValue:bool))
   > ChangeRewardView(InItemView=false, OnlyVisual=false)
   Event On Selected()
   > SetVisibility(self=Get SelectedStateOutline() -> (SelectedStateOutline:Image Object Reference), InVisibility=SelfHitTestInvisible)
   > RequestTransitionOnBlocks!(execute, self:UIKit Modular Button Object Reference, State:EUIKitBlockVisualState Enum=Selected) -> (then)
   > UpdateContentVisibility()
   n27 = Get ListView()
   n28 = OR Boolean!(
    A=IsValid(Object=Get CurrentSelectedItem() -> (CurrentSelectedItem:Common Button Base Object Reference))
    B=IsHovered(self=Get finalGoalReward())
    C:bool=IsValid(Object=Get Selected Item(self=@n27))
   ) -> (ReturnValue)
   > Branch(Condition=AND Boolean(A=NOT Boolean(A=@n28), B=Get Is Item View()))
   > ChangeSelectedIndex(Index=0)
   Event On Deselected()
   > SetVisibility(self=Get SelectedStateOutline(), InVisibility=Collapsed)
   > ClearAllSelection() {is=title K2Node_CallFunction}
   > SetBottomButtonVisibility(self=Get DecoyButtonBack(), ShouldButtonBeVisible=false)
   n29 = GetListItemObject!(UserObjectListEntry:User Object List Entry Interface) -> (ReturnValue:Object Reference)
   n30 = GetIsItemView()
   > n31 = Branch(Condition=AND Boolean(A=IsValid(Object=@n29), B=NOT Boolean(A=@n30)))
   > n32 = RequestTransitionOnBlocks(State=Select(Option 0:EUIKitBlockVisualState Enum=DeselectedIdle, Option 1:EUIKitBlockVisualState Enum=DeselectedFocused, Index=IsHovered()) -> (ReturnValue:EUIKitBlockVisualState Enum))
   Branch(execute=[@n32.then, @n33.then], Condition=IsCustomWidget())
   > ChangeRewardView(InItemView=false, OnlyVisual=false)
   > UpdateContentVisibility()
   n33 = InstantDeselect(execute=@n31.else)
   n34 = Event OnListItemObjectSet()
   > RequestInstantTransition(self=Get Block_Outline(), State=InstantUnfocused)
   > n35 = Cast To QuestUIHorizontalPunchCardVM!(execute, Object=UnwrapItem!(Item:Object Reference=@n34.ListItemObject) -> (ReturnValue:Object Reference)) -> (then, CastFailed, AsQuest UI Horizontal Punch Card View Model:Quest UI Horizontal Punch Card View Model Object Reference) {is=title K2Node_DynamicCast}
   > Set QuestUIHorizontalPunchCardVM(QuestUIHorizontalPunchCardVM:Quest UI Horizontal Punch Card View Model Object Reference=@n35) -> (Output_Get:Quest UI Horizontal Punch Card View Model Object Reference)
   > ChangeRewardView(InItemView=NOT Boolean(A=IsCustomWidget()), OnlyVisual=true)
   > UpdateContentVisibility()
   > UpdateHintVisibility()
   > UpdateOutlineVisibility()
   Event PreConstruct()
   > Parent: PreConstruct(IsDesignTime=false)
   > SetBlocks()
   > SetColorAndOpacity(
    self:User Widget Object Reference=[
     Get Block_Icon!(self:WBP UIKit Button Generic Base Object Reference=Get GoalsHorizontal_PreviousButton() -> (GoalsHorizontal_PreviousButton:WBP UIKit Button Generic Object Reference)) -> (Block_Icon:WBP UIKit Block Image Object Reference)
     Get Block_Icon(self=Get GoalsHorizontal_NextButton() -> (GoalsHorizontal_NextButton:WBP UIKit Button Generic Object Reference))
    ]
    InColorAndOpacity:Linear Color Structure=(R=0,G=0,B=0,A=1)
   )
   > SetScrollbarVisibility!(execute, self:List View Base Object Reference=Get ListView(), InVisibility:ESlateVisibility Enum=Collapsed) -> (then)
   > SetCompactUIStyle() {is=title K2Node_CallFunction}
   Event On Double Clicked() {is=title K2Node_Event}
   > Call On Goals Double Click!(
    execute:Exec
    self:WBP Goals Progress Horizontal Object Reference
    GoalsQuest:WBP Goals Progress Horizontal Object Reference=Self-Reference()
   ) -> (then:Exec) {is=title K2Node_CallDelegate}
   Event On Unfocused() {is=title K2Node_Event}
   > n36 = Branch(Condition=IsHovered())
   Parent: On Unfocused(execute=@n36.else) {is=title K2Node_CallParentFunction}
   Event On Focused() {is=title K2Node_Event}
   > n37 = Parent: On Focused() {is=title K2Node_CallParentFunction}
  ]}
 FillConteiners {nodes [
   FillConteiners() {is=title K2Node_FunctionEntry}
   > n1 = Switch on E_QuestPunchCardState(Selection=Get CurrentState() -> (CurrentState=NewEnumerator0))
   n2 = Get GoalRewardVMs() -> (GoalRewardVMs:Array of Quest UIGoal Reward VM Object References)
   n3 = For Each Loop(Exec=@n1.NewEnumerator1, Array:Array of Quest UIGoal Reward VM Object References=@n2) -> (Array Element:Quest UIGoal Reward VM Object Reference)
   > n4 = Create Entry of Class!(
    execute
    self:Dynamic Entry Box Object Reference=Get DynamicEntryBox_Fill()
    EntryClass:User Widget Class Reference=&/WBP_GoalReward_Container.WBP_GoalReward_Container_C
   ) -> (then, ReturnValue:WBP Goal Reward Container Object Reference)
   > SetQuestVM!(execute, self:WBP Goal Reward Container Object Reference=@n4, QuestUIGoalRewardVM:Quest UIGoal Reward VM Object Reference=@n3) -> (then)
   > Bind Event to On Clicked(self:WBP Goal Reward Container Object Reference=@n4, Delegate=Create Event()) {is=title K2Node_AddDelegate}
   > Bind Event to On Selected(self:WBP Goal Reward Container Object Reference=@n4, Delegate=Create Event()) {is=title K2Node_AddDelegate}
   n5 = Find(
    TargetMap:"Map of Integers to Float (double-precision)s"=Get DynamicEntryBoxPaddings() -> (DynamicEntryBoxPaddings:"Map of Integers to Float (double-precision)s")
    Key:"Integer (by ref)"=Length(TargetArray:Array of Quest UIGoal Reward VM Object References=Get GoalRewardVMs())
   ) -> (Value:double)
   n6 = SlotAsOverlaySlot(Widget=Get DynamicEntryBox_Fill())
   n7 = Get!(self:Overlay Slot Object Reference=@n6) -> (Padding_Left:float=~, Padding_Top:float=~, Padding_Right:float=~, Padding_Bottom:float=~)
   > SetPadding!(
    execute
    self:Overlay Slot Object Reference=@n6
    InPadding_Left:float=@n5
    InPadding_Top:float=@n7.Padding_Top
    InPadding_Right:float=@n5
    InPadding_Bottom:float=@n7.Padding_Bottom
   ) -> (then)
   n8 = Length(TargetArray=@n2)
   > n9 = Branch(Condition="Equal (Integer)"(A=@n3.Array Index, B=int - int(A=@n8)))
   > SetNavigationRuleCustom!(execute, self:Widget Object Reference=@n4, Direction:EUINavigation Enum=Right, InCustomDelegate:Delegate=Create Event()) -> (then)
   SetNavigationRuleBase!(execute=@n9.else, self:Widget Object Reference=@n4, Direction:EUINavigation Enum=Right, Rule:EUINavigationRule Enum=Escape) -> (then)
   Set List Items!(execute=@n1.NewEnumerator0, self:List View Object Reference=Get ListView(), InListItems:Array of Object References=Get GoalRewardVMs()) -> (then) {is=title K2Node_CallFunction}
  ]}
 GetIsBundleView {nodes [
   GetIsBundleView() {is=title K2Node_FunctionEntry}
   n1 = Get Is Item View()
   > Return Node!(execute, ReturnValue:bool=AND Boolean(A=IsCustomWidget(), B=NOT Boolean(A=@n1)))
  ]}
 GetIsItemView {nodes [
   GetIsItemView() {is=title K2Node_FunctionEntry}
   > Return Node(ReturnValue=AND Boolean(A=IsCustomWidget(), B=Get Is Item View()))
  ]}
 GetLastRewardElement {nodes [
   n1 = Inputs!() -> (Array:Array of Quest UIGoal Reward VM Object References)
   n2 = Length(TargetArray=@n1)
   n3 = int - int(A=@n2)
   Outputs!(Element:Quest UIGoal Reward VM Object Reference=Select(
     Option 0:Quest UIGoal Reward VM Object Reference
     Option 1:Quest UIGoal Reward VM Object Reference="Get (a copy)"(Array:Array of Quest UIGoal Reward VM Object References=@n1, Dimension 1=@n3) -> (Output:Quest UIGoal Reward VM Object Reference)
     Index="integer > integer"(A=@n2)
    ) -> (ReturnValue:Quest UIGoal Reward VM Object Reference))
  ]}
 HasExpirationTime {nodes [
   n1 = HasExpirationTime!() -> (then, HasExpirationTime:bool=~) {is=title K2Node_FunctionEntry}
   > SetVisibility(
    self=[
     Get ExpirationTime() -> (ExpirationTime:Athena Challenge Date Time Text Block Object Reference)
     Get TimerBack() -> (TimerBack:Image Object Reference)
    ]
    InVisibility=Select(Option 0:ESlateVisibility Enum=Collapsed, Option 1:ESlateVisibility Enum=HitTestInvisible, Index=@n1) -> (ReturnValue:ESlateVisibility Enum)
   )
  ]}
 InstantDeselect {nodes [
   InstantDeselect() {is=title K2Node_FunctionEntry}
   > ClearAllSelection()
   > SetScrollOffset!(execute, self:List View Base Object Reference=Get ListView(), InScrollOffset:float=0) -> (then)
   > RequestInstantTransitionOnBlocks!(execute, self:UIKit Modular Button Object Reference, State:EUIKitBlockInstantTransitionState Enum=InstantDeselected) -> (then)
   > SetProgressBarType(self=Get finalGoalReward(), WithSeparator=true)
  ]}
 IsCustomWidget {nodes [
   IsCustomWidget() {is=title K2Node_FunctionEntry}
   > n1 = Get!(execute:Exec) -> (then:Exec, else:Exec, QuestUIHorizontalPunchCardVM:Quest UI Horizontal Punch Card View Model Object Reference)
   > Return Node!(
    execute
    HasCustom:bool=IsValidSoftClassReference!(SoftClassReference:"Object Soft Class Reference (by ref)"=Get CustomMapWidget!(self:Quest UI Punch Card View Model Object Reference=@n1) -> (CustomMapWidget:User Widget Soft Class Reference)) -> (ReturnValue:bool)
   )
  ]}
 IsFinalGoalVisible {nodes [
   IsFinalGoalVisible() {is=title K2Node_FunctionEntry}
   > n1 = Get()
   n2 = Get OrderedContentVMs(self=@n1)
   > Return Node!(
    execute
    ReturnValue:bool="integer > integer"(
     A=Length(TargetArray:Array of Quest UI Base Content View Model Object References=@n2)
     B=Get ItemCountScrollThreshold() -> (ItemCountScrollThreshold=0)
    )
   )
  ]}
 IsFinalRewardVisible {nodes [
   IsFinalRewardVisible() {is=title K2Node_FunctionEntry}
   n1 = GetVisibility!(self:Widget Object Reference=Get finalGoalReward()) -> (ReturnValue:ESlateVisibility Enum=Visible)
   n2 = GetActiveWidget!(self:Common Visibility Switcher Object Reference=Get ListSwitcher() -> (ListSwitcher:Common Visibility Switcher Object Reference)) -> (ReturnValue:Widget Object Reference)
   n3 = Get ListViewRoot() -> (ListViewRoot:Grid Panel Object Reference)
   > Return Node(ReturnValue=AND Boolean(A="Not Equal (Enum)"(A:ESlateVisibility Enum=@n1, B:ESlateVisibility Enum=Collapsed), B="Equal (Object)"(A=@n2, B=@n3)))
  ]}
 IsFrontend {nodes [
   IsFrontend() {is=title K2Node_FunctionEntry}
   > n1 = Cast To FrontEnd_PlayerController!(execute, Object=GetOwningPlayer()) -> (then, CastFailed, AsFront End Player Controller:Front End Player Controller Object Reference) {is=title K2Node_DynamicCast}
   > Return Node!(execute, IsFrontend:bool=true)
   Return Node(execute=@n1.CastFailed, IsFrontend=false)
  ]}
 IsGamepad {nodes [
   IsGamepad() {is=title K2Node_FunctionEntry}
   n1 = Get CommonInputSubsystem()
   > Return Node!(execute, ReturnValue:bool="Equal (Enum)"(A=GetCurrentInputType(self=@n1), B=Gamepad))
  ]}
 IsNew {nodes [
   n1 = IsNew!() -> (then, bIsNew:bool=~) {is=title K2Node_FunctionEntry}
   > SetVisibility(
    self=Get StatusIndicator() -> (StatusIndicator:WBP UIKit Status Indicator Object Reference)
    InVisibility=Select(Option 0=Hidden, Option 1=SelfHitTestInvisible, Index=@n1)
   )
  ]}
 MarkPunchCardSeen {nodes [
   MarkPunchCardSeen() {is=title K2Node_FunctionEntry}
   > n1 = For Each Loop(Array=Get GoalRewardVMs())
   MarkSeen!(execute=@n1.Completed, self:Quest UI Base View Model Object Reference=Get QuestUIHorizontalPunchCardVM(), bCommitToMcp:bool=false) -> (then)
   MarkSeen(
    execute=@n1.LoopBody
    self=Get QuestVM!(self:Quest UIGoal Reward VM Object Reference=@n1) -> (QuestVM:Quest UI Quest View Model Object Reference)
    bCommitToMcp=false
   )
  ]}
 OnGoalsDoubleClick {nodes [OnGoalsDoubleClick!() -> (then, GoalsQuest:WBP Goals Progress Horizontal Object Reference) {is=title K2Node_FunctionEntry}]}
 OnGoalsNavigate {nodes [OnGoalsNavigate!() -> (then, GoalsProgressWidget:User Widget Object Reference, Direction:EUINavigation Enum) {is=title K2Node_FunctionEntry}]}
 OnInputChanged {nodes [
   n1 = OnInputChanged!() -> (then, bNewInputType:ECommonInputType Enum) {is=title K2Node_FunctionEntry}
   > UpdateHintVisibility()
   > SetVisibility(
    self=Get FocusCatcher() -> (FocusCatcher:Button Object Reference)
    InVisibility=Select(Option 0=HitTestInvisible, Option 1=Visible, Index="Equal (Enum)"(A=@n1, B=Gamepad))
   )
   > UpdateFocus()
  ]}
 OnRewardViewChanged {nodes [OnRewardViewChanged!() -> (then, InItemView:bool=~) {is=title K2Node_FunctionEntry}]}
 OnSelectedIndexChanged {nodes [OnSelectedIndexChanged!() -> (then, Index:int) {is=title K2Node_FunctionEntry}]}
 PopulateListView {nodes [
   n1 = PopulateListView!() -> (then, InRewards:Array of Object References) {is=title K2Node_FunctionEntry}
   > Clear(TargetArray:Array of Quest UIGoal Reward VM Object References=Get GoalRewardVMs())
   > Clear(TargetArray:Array of Quest UI Reward VM Object References=Get RewardVMs() -> (RewardVMs:Array of Quest UI Reward VM Object References))
   n2 = "integer > integer"(A=Length(TargetArray:Array of Object References=@n1), B=1)
   > UpdateDescriptionText!(execute, self, MoreThenOne:bool=@n2) -> (then)
   > n3 = Branch(Condition=@n2)
   n4 = IsFinalGoalVisible() {is=title K2Node_CallFunction}
   > Set CurrentState(CurrentState:E_QuestPunchCardState Enum=Select(Option 0:E_QuestPunchCardState Enum=NewEnumerator1, Option 1:E_QuestPunchCardState Enum=NewEnumerator0, Index=@n4) -> (ReturnValue:E_QuestPunchCardState Enum)) -> (Output_Get:E_QuestPunchCardState Enum=NewEnumerator0)
   > SetActiveWidget(
    self=Get ListSwitcher()
    Widget=Select(Option 0:Widget Object Reference=Get DynamicEntryBox_Fill(), Option 1:Widget Object Reference=Get ListViewRoot(), Index=@n4) -> (ReturnValue:Widget Object Reference)
   )
   > Reset!(execute, self:Dynamic Entry Box Object Reference=Get DynamicEntryBox_Fill(), bDeleteWidgets:bool=true) -> (then)
   > n5 = ClearListItems(self=Get ListView())
   n6 = For Each Loop(Exec=[@n5.then, @n16.then], Array:Array of Object References=Get InRewards() -> (InRewards:Array of Object References)) -> (Array Element:Object Reference)
   n7 = Branch(
    execute=@n6.Completed
    Condition="Not Equal (Enum)"(A:E_QuestPunchCardState Enum=Get CurrentState() -> (CurrentState=NewEnumerator0), B:E_QuestPunchCardState Enum=NewEnumerator2)
   )
   > SetUpFinalGoal()
   > FillConteiners() {is=title K2Node_CallFunction}
   > Refresh()
   > n8 = UpdateNextItemProgressStyle()
   OnInputChanged!(execute=[@n8.then, @n9.then], self, bNewInputType:ECommonInputType Enum=GetCurrentInputType(self=Get CommonInputSubsystem())) -> (then) {is=title K2Node_CallFunction}
   SetQuestVM(
    execute=@n7.else
    self:WBP Goal Reward Solo Object Reference=Get WBP_GoalReward_Solo() -> (WBP_GoalReward_Solo:WBP Goal Reward Solo Object Reference)
    QuestUIGoalRewardVM="Get (a copy)"(Array=Get GoalRewardVMs())
   )
   > n9 = Refresh()
   Set ArrayIndex(execute=@n6.LoopBody, ArrayIndex:int=@n6.Array Index) -> (Output_Get:int=0)
   > n10 = Cast To QuestUIQuestVM!(execute, Object=@n6) -> (then, CastFailed, AsQuest UI Quest View Model:Quest UI Quest View Model Object Reference) {is=title K2Node_DynamicCast}
   > Set QuestUIReward(QuestUIReward:Quest UI Quest View Model Object Reference=@n10) -> (Output_Get:Quest UI Quest View Model Object Reference)
   n11 = "Equal (Enum)"(
    A:EGoalCardOrientation Enum=Get Orientation!(self:Quest UI Punch Card View Model Object Reference=Get QuestUIHorizontalPunchCardVM()) -> (Orientation:EGoalCardOrientation Enum=Vertical)
    B:EGoalCardOrientation Enum=Horizontal
   )
   n12 = Get ArrayIndex() -> (ArrayIndex:int=0)
   > n13 = Construct Object from Class!(
    execute
    Class=&/QuestUIGoalRewardVM.QuestUIGoalRewardVM_C
    self
    Index:int=@n12
    QuestVM:Quest UI Quest View Model Object Reference=@n10
    ShowProgress:bool=@n11
    IsLastIndex:bool=False
   ) -> (then, ReturnValue:Quest UIGoal Reward VM Object Reference)
   > Add(
    TargetArray:Array of Quest UIGoal Reward VM Object References=Get GoalRewardVMs()
    NewItem:"Quest UIGoal Reward VM Object Reference (by ref)"=@n13
   ) {is=title K2Node_CallArrayFunction}
   > Set with Broadcast Extension!(
    execute
    Extension:Quest UIHorizontal Punch Card Extension Object Reference=Get Extension By Type!(
     self:Quest UI Base View Model Object Reference=Get QuestUIReward() -> (QuestUIReward:Quest UI Quest View Model Object Reference)
     ExtensionClass:Quest UIExtension Base Class Reference=/Script/QuestUI.QuestUIHorizontalPunchCardExtension
    ) -> (ReturnValue:Quest UIHorizontal Punch Card Extension Object Reference) {is=title K2Node_CallFunction}
    self:Quest UIGoal Reward VM Object Reference=@n13
   ) -> (then, Output_Get:Quest UIHorizontal Punch Card Extension Object Reference)
   > n14 = For Each Loop(Array:Array of Quest UI Reward VM Object References=Get RewardVMs!(self:Quest UI Quest View Model Object Reference=Get QuestUIReward()) -> (RewardVMs)) -> (Array Element:Quest UI Reward VM Object Reference)
   > Add(
    TargetArray:Array of Quest UI Reward VM Object References=Get RewardVMs!() -> (RewardVMs)
    NewItem:"Quest UI Reward VM Object Reference (by ref)"=@n14
   )
   n15 = Construct Object from Class(
    execute=@n10.CastFailed
    Class=&QuestUIGoalRewardVM_C
    Index=@n12
    ShowProgress=@n11
    IsLastIndex=False
   )
   > Add(
    TargetArray:Array of Quest UIGoal Reward VM Object References=Get GoalRewardVMs()
    NewItem:"Quest UIGoal Reward VM Object Reference (by ref)"=@n15
   )
   Set CurrentState(execute=@n3.else, CurrentState=NewEnumerator2) -> (Output_Get=NewEnumerator0)
   > n16 = SetActiveWidget(self=Get ListSwitcher(), Widget=Get WBP_GoalReward_Solo())
  ]}
 Refresh {nodes [
   Refresh() {is=title K2Node_FunctionEntry}
   > Set TotalRequiredCount(TotalRequiredCount:int=0) -> (Output_Get:int=0)
   > Set CurrentQuestVM(CurrentQuestVM:Quest UI Quest View Model Object Reference) -> (Output_Get:Quest UI Quest View Model Object Reference)
   > n1 = Set LocalGoalRewardVMs(LocalGoalRewardVMs:Array of Quest UIGoal Reward VM Object References=Get GoalRewardVMs()) -> (Output_Get:Array of Quest UIGoal Reward VM Object References)
   n2 = Get Last Reward Element!(Array:Array of Quest UIGoal Reward VM Object References=@n1) -> (Element:Quest UIGoal Reward VM Object Reference) {is=title K2Node_MacroInstance}
   n3 = Get QuestUIGoalRewardVM(self:WBP Goal Reward Object Reference=Get finalGoalReward())
   > n4 = Branch(Condition=AND Boolean(A=IsFinalGoalVisible(), B="Not Equal (Object)"(A=@n2, B=@n3) -> (ReturnValue=~)))
   > n5 = Add(TargetArray=Get LocalGoalRewardVMs() -> (LocalGoalRewardVMs:Array of Quest UIGoal Reward VM Object References), NewItem=@n3)
   n6 = Get QuestUIHorizontalPunchCardVM()
   n7 = Branch(execute=[@n5.then, @n4.else], Condition="Equal (Enum)"(A=Get Orientation(self=@n6) -> (Orientation=Vertical), B=Horizontal))
   > n8 = For Each Loop with Break!(
    Exec:Exec
    Array:Array of Quest UIGoal Reward VM Object References=Get LocalGoalRewardVMs()
    Break:Exec=[@n20.Is Valid, @n22.else, @n23.then]
   ) -> (LoopBody:Exec, Array Element:Quest UIGoal Reward VM Object Reference, Array Index:int, Completed:Exec)
   n9 = Is Valid(exec=@n8.Completed, InputObject=Get CurrentQuestVM() -> (CurrentQuestVM:Quest UI Quest View Model Object Reference))
   n10 = Get Last Reward Element(Array=Get LocalGoalRewardVMs())
   n11 = Is Valid(exec=@n9.Is Not Valid, InputObject=@n10)
   > n12 = Set CurrentQuestVM(CurrentQuestVM=Get QuestVM(self=@n10))
   UpdateObjectiveInfo!(
    execute=[@n9.Is Valid, @n12.then, @n11.Is Not Valid]
    self
    QuestVM:Quest UI Quest View Model Object Reference=Get CurrentQuestVM()
    CanShowOverallProgress:bool=false
   ) -> (then)
   n13 = Get QuestVM(self=@n8)
   n14 = Get RequiredCount!(self:Quest UI Quest View Model Object Reference=@n13) -> (RequiredCount:int=0)
   n15 = Get AchievedCount!(self:Quest UI Quest View Model Object Reference=@n13) -> (AchievedCount:int=0)
   n16 = Branch(execute=@n8.LoopBody, Condition=AND Boolean(A="integer >= integer"(A=@n15, B=@n14), B="integer > integer"(A=@n14)))
   > Set TotalRequiredCount(TotalRequiredCount=@n14) -> (Output_Get=0)
   > Set with Broadcast ProgressValue!(execute, ProgressValue:double=1, self:Quest UIGoal Reward VM Object Reference=@n8) -> (then, Output_Get:double)
   n17 = Get QuestVM(self=@n8)
   n18 = int - int(A=Get RequiredCount(self=@n17) -> (RequiredCount=0), B=Get TotalRequiredCount() -> (TotalRequiredCount:int=0))
   n19 = int - int(A=Get AchievedCount(self=@n17) -> (AchievedCount=0), B=Get TotalRequiredCount() -> (TotalRequiredCount=0))
   Set with Broadcast ProgressValue(execute=@n16.else, ProgressValue=float / float(A:int="Max (Integer)"(A=@n19, B=0), B:int="Max (Integer)"(A=@n18, B=0)), self=@n8)
   > Set with Broadcast NextReward!(execute, NextReward:bool=true, self:Quest UIGoal Reward VM Object Reference=@n8) -> (then, Output_Get:bool)
   > n20 = Is Valid(InputObject=Get CurrentQuestVM())
   n21 = Set CurrentQuestVM(execute=@n20.Is Not Valid, CurrentQuestVM=Get QuestVM(self=@n8))
   > n22 = Branch(Condition="Not Equal (Object)"(A=@n21, B=Get As Quest UIGoal Reward VM()) -> (ReturnValue=~))
   > n23 = Set QuestUIQuestVM(QuestUIQuestVM:Quest UI Quest View Model Object Reference=@n21) -> (Output_Get:Quest UI Quest View Model Object Reference)
   n24 = For Each Loop(Exec=@n7.else, Array:Array of Quest UIGoal Reward VM Object References=Get LocalGoalRewardVMs()) -> (Array Element:Quest UIGoal Reward VM Object Reference)
   UpdateObjectiveInfo(execute=@n24.Completed, CanShowOverallProgress=true)
   n25 = Get QuestVM(self=@n24)
   n26 = Set with Broadcast ProgressValue(
    execute=@n24.LoopBody
    ProgressValue=float / float(A=Get AchievedCount(self=@n25) -> (AchievedCount=0), B=Get RequiredCount(self=@n25) -> (RequiredCount=0))
    self=@n24
   )
   n27 = IsValid(Object=Get CurrentQuestVM())
   > Set with Broadcast NextReward(NextReward=AND Boolean(A="float < float"(A=@n26, B=1), B=NOT Boolean(A=@n27)), self=@n24)
   > n28 = Is Valid(InputObject=Get CurrentQuestVM())
   Set CurrentQuestVM(execute=@n28.Is Not Valid, CurrentQuestVM=Get QuestVM(self=@n24))
  ]}
 SelectAdjacentEntry {nodes [
   SelectAdjacentEntry!() -> (then, Right:bool=~) {is=title K2Node_FunctionEntry}
   n1 = Get SelectedIndex() -> (SelectedIndex:int=0)
   > Set TargetIndex(TargetIndex:int=Select(Option 0:int=int - int(A=@n1), Option 1:int=int + int(A=@n1), Index=Get Right() -> (Right:bool)) -> (ReturnValue:int)) -> (Output_Get:int=0)
   n2 = Get QuestUIHorizontalPunchCardVM()
   n3 = int - int(A=Length(TargetArray:Array of Quest UI Base Content View Model Object References=Get OrderedContentVMs(self=@n2)))
   n4 = Get TargetIndex() -> (TargetIndex:int=0)
   n5 = "integer < integer"(A=Get TargetIndex() -> (TargetIndex=0))
   n6 = Get TargetIndex() -> (TargetIndex=0)
   > ChangeSelectedIndex(Index=Select(Option 0=Select(Option 0=@n6, Option 1=@n3, Index=@n5), Option 1=0, Index="integer > integer"(A=@n4, B=@n3)))
   > ChangeRewardView(InItemView=true, OnlyVisual=false)
  ]}
 SelectFinalGoal {nodes [
   n1 = SelectFinalGoal!() -> (then, Navigation:EUINavigation Enum) {is=title K2Node_FunctionEntry}
   > n2 = Switch on EUINavigation!(execute, Selection:EUINavigation Enum=@n1) -> (Left:Exec, Right:Exec, Up:Exec, Down:Exec, Next:Exec, Previous:Exec, Invalid:Exec) {is=title K2Node_SwitchEnum}
   n3 = Get finalGoalReward()
   SelectFinalGoalReward(execute=@n2.Right, Widget=@n3)
   > Return Node(ReturnValue:Widget Object Reference=@n3)
   n4 = Get ListView()
   SetSelectedIndex(execute=@n2.Left, self=@n4, Index=int - int(A=GetNumItems(self=@n4) -> (ReturnValue=0)))
   > ClearSelection(self=Get finalGoalReward())
   > Return Node(ReturnValue=@n4)
  ]}
 SelectFinalGoalReward {nodes [
   SelectFinalGoalReward!() -> (then, Widget:Common Button Base Object Reference) {is=title K2Node_FunctionEntry}
   n1 = Get QuestUIHorizontalPunchCardVM()
   > ChangeSelectedIndex(Index=Last Index!(TargetArray:Array of Quest UI Base Content View Model Object References=Get OrderedContentVMs(self=@n1)) -> (ReturnValue=0))
  ]}
 SelectFirstElement {nodes [
   SelectFirstElement!() -> (then, Navigation:EUINavigation Enum) {is=title K2Node_FunctionEntry}
   > ChangeSelectedIndex(Index=0)
   > Return Node()
  ]}
 SetBlocks {nodes [
   SetBlocks() {is=title K2Node_FunctionEntry}
   > RegisterModularBlock!(execute, self:UIKit Modular Button Object Reference, Block:UIKit Block Object Reference=Get Block_Outline()) -> (then)
   > SetBlockTiming!(
    execute
    self:UIKit Block Object Reference=Get Block_Outline()
    Timing:"UIKit Block Timing Structure (by ref)"=Get BlockTiming() -> (BlockTiming:UIKit Block Timing Structure)
   ) -> (then)
  ]}
 SetCompactUIStyle {nodes [
   SetCompactUIStyle() {is=title K2Node_FunctionEntry}
   > Branch(Condition=IsUsingCompactUI())
   n1 = Get GoalsHorizontal_PreviousButton()
   > SetIconSize!(execute, self:WBP UIKit Button Generic Base Object Reference=@n1, Icon:Vector 2D Structure=(X=48,Y=48)) -> (then)
   > SetButtonPadding!(
    execute
    self:WBP UIKit Button Generic Base Object Reference=@n1
    Padding_Left:float=0
    Padding_Top:float=0
    Padding_Right:float=2
    Padding_Bottom:float=0
   ) -> (then)
   n2 = Get GoalsHorizontal_NextButton()
   > SetIconSize(self=@n2, Icon=(X=48,Y=48))
   > SetButtonPadding(self=@n2, Padding_Left=2, Padding_Top=0, Padding_Right=0, Padding_Bottom=0)
   n3 = Get Buttons() -> (Buttons:Grid Panel Object Reference)
   n4 = Get!(self:Grid Slot Object Reference=SlotAsGridSlot(Widget=@n3)) -> (Padding_Left:float=~, Padding_Top:float=~, Padding_Right:float=~, Padding_Bottom:float=~)
   > SetGridSlotPadding!(
    execute
    Widget:Widget Object Reference=@n3
    Padding_Left:float=-30
    Padding_Top:float=@n4.Padding_Top
    Padding_Right:float=-30
    Padding_Bottom:float=@n4.Padding_Bottom
   ) -> (then)
  ]}
 SetContentEnabled {nodes [
   n1 = SetContentEnabled!() -> (then, bEnabled:bool=~) {is=title K2Node_FunctionEntry}
   > SetRenderOpacity(
    self=[Get Subheader() -> (Subheader:Horizontal Box Object Reference), Get ListSwitcher(), Get Subheader()]
    InOpacity=Select(Option 0:float=0.5, Option 1:float=1, Index=@n1) -> (ReturnValue:float)
   )
   > SetRenderOpacity(self=Get ListViewProgressPatch() -> (ListViewProgressPatch:Image Object Reference), InOpacity=Select(Option 0=2, Option 1=1, Index=@n1))
  ]}
 SetDebugBundleID {nodes [
   n1 = SetDebugBundleID!() -> (then, BundleID:"String (by ref)") {is=title K2Node_FunctionEntry}
   > "SetText (Text)"(self=Get DebugBundleID() -> (DebugBundleID:Common Text Object Reference), InText="To Text (String)"(InString=@n1))
  ]}
 SetExpirationTime {nodes [
   n1 = SetExpirationTime!() -> (then, ExpirationTime:Date Time Structure) {is=title K2Node_FunctionEntry}
   > SetDateTimeValue!(
    execute
    self:Common Date Time Text Block Object Reference=Get ExpirationTime()
    InDateTime:Date Time Structure=@n1
    bShowAsCountdown:bool=true
    InRefreshDelay:float=1
   ) -> (then)
  ]}
 SetNextRewardSelected {nodes [
   SetNextRewardSelected() {is=title K2Node_FunctionEntry}
   > n1 = Switch on E_QuestPunchCardState(Selection=Get CurrentState() -> (CurrentState=NewEnumerator0))
   > n2 = For Each Loop with Break(
    Array:Array of Object References=GetListItems!(self:List View Object Reference=Get ListView()) -> (ReturnValue:Array of Object References)
    Break=@n4.then
   ) -> (Array Element:Object Reference)
   ChangeSelectedIndex(execute=@n2.Completed, Index=Get NextRewardIndex() -> (NextRewardIndex:int=0))
   n3 = Cast To QuestUIGoalRewardVM(execute=@n2.LoopBody, Object=@n2)
   > Branch(Condition=Get NextReward!(self:Quest UIGoal Reward VM Object Reference=@n3) -> (NextReward:bool))
   > n4 = Set NextRewardIndex(NextRewardIndex:int=@n2.Array Index) -> (Output_Get:int=0)
   n5 = For Each Loop with Break(Exec=@n1.NewEnumerator1, Array:Array of User Widget Object References=GetAllEntries(self=Get DynamicEntryBox_Fill()), Break=@n7.then) -> (Array Element:User Widget Object Reference)
   > n6 = Cast To WBP_GoalReward(Object=@n5)
   > Branch(Condition=Get NextReward(self=Get QuestUIGoalRewardVM(self=@n6)))
   > n7 = ChangeSelectedIndex(Index=@n5.Array Index)
  ]}
 SetUpFinalGoal {nodes [
   SetUpFinalGoal() {is=title K2Node_FunctionEntry}
   n1 = Length(TargetArray:Array of Quest UIGoal Reward VM Object References=Get GoalRewardVMs())
   > SetGridSlotAlignment!(
    execute
    Widget:Widget Object Reference=Get ListSwitcher()
    Horizontal:EHorizontalAlignment Enum=Select(Option 0:EHorizontalAlignment Enum=HAlign_Fill, Option 1:EHorizontalAlignment Enum=HAlign_Center, Index="Equal (Integer)"(A=@n1, B=1)) -> (ReturnValue:EHorizontalAlignment Enum)
    Vertical:EVerticalAlignment Enum=VAlign_Fill
   ) -> (then)
   > SetVisibility(
    self=Get finalGoalReward()
    InVisibility=Select(Option 0:ESlateVisibility Enum=Collapsed, Option 1:ESlateVisibility Enum=SelfHitTestInvisible, Index=IsFinalGoalVisible()) -> (ReturnValue:ESlateVisibility Enum)
   )
   > n2 = Branch(Condition=NOT Boolean(A=IsFinalGoalVisible()))
   > SetProgressBarType(self=Get finalGoalReward(), WithSeparator=false)
   n3 = Get GoalRewardVMs()
   > Set with Broadcast IsLastIndex!(
    execute
    IsLastIndex:bool=true
    self:Quest UIGoal Reward VM Object Reference="Get (a copy)"(Array=@n3, Dimension 1=Last Index(TargetArray:Array of Quest UIGoal Reward VM Object References=@n3) -> (ReturnValue=0))
   ) -> (then, Output_Get:bool)
   n4 = Get GoalRewardVMs()
   n5 = "Get (a copy)"(Array=@n4, Dimension 1=int - int(A=Length(TargetArray=@n4)))
   Set QuestUIGoalRewardVM!(
    execute=@n2.else
    QuestUIGoalRewardVM:Quest UIGoal Reward VM Object Reference=@n5
    self:WBP Goal Reward Object Reference=Get finalGoalReward()
   ) -> (then, Output_Get:Quest UIGoal Reward VM Object Reference)
   > Remove Item!(
    execute
    TargetArray:Array of Quest UIGoal Reward VM Object References=Get GoalRewardVMs()
    Item:"Quest UIGoal Reward VM Object Reference (by ref)"=@n5
   ) -> (then, ReturnValue:bool)
   n6 = Get finalGoalReward()
   > SetProgressBarType(self=@n6, WithSeparator=true)
   > Set with Broadcast IsLastIndex(IsLastIndex=true, self=Get QuestUIGoalRewardVM(self=@n6))
  ]}
 UpdateContentVisibility {nodes [
   UpdateContentVisibility() {is=title K2Node_FunctionEntry}
   n1 = GetSelected()
   n2 = Select(
    Option 0=SelfHitTestInvisible
    Option 1=HitTestInvisible
    Index=AND Boolean(A=NOT Boolean(A=@n1), B=IsCustomWidget() {is=title K2Node_CallFunction})
   )
   > SetVisibility(self=Get ListSwitcher(), InVisibility=@n2)
   n3 = Get Buttons()
   > Branch(Condition=IsVisible(self=@n3))
   > SetVisibility(self=@n3, InVisibility=@n2)
  ]}
 UpdateDescriptionText {nodes [
   n1 = UpdateDescriptionText!() -> (then, MoreThenOne:bool=~) {is=title K2Node_FunctionEntry}
   n2 = SlotAsGridSlot(Widget=Get Subheader())
   n3 = Get(self=@n2) -> (Padding_Left=~, Padding_Top=~, Padding_Right=~, Padding_Bottom=~)
   > SetPadding(
    self:Grid Slot Object Reference=@n2
    InPadding_Left=@n3
    InPadding_Top=@n3.Padding_Top
    InPadding_Right=Select(Option 0:float=180, Option 1:float=24, Index=@n1) -> (ReturnValue:float)
    InPadding_Bottom=@n3.Padding_Bottom
   )
   n4 = Get MoreThenOne() -> (MoreThenOne:bool)
   > SetRow!(
    execute
    self:Grid Slot Object Reference=[SlotAsGridSlot(Widget=Get FocusCatcher()), @n2]
    InRow:int="To Integer (Boolean)"!(InBool:bool=NOT Boolean(A=@n4)) -> (ReturnValue:int=0)
   ) -> (then)
  ]}
 UpdateFocus {nodes [
   UpdateFocus() {is=title K2Node_FunctionEntry}
   n1 = GetCurrentInputType(self=Get CommonInputSubsystem())
   > Branch(Condition=AND Boolean!(A="Equal (Enum)"(A:ECommonInputType Enum=@n1, B:ECommonInputType Enum=Gamepad), B=IsHovered(), C:bool=GetSelected()) -> (ReturnValue))
   > n2 = Branch(Condition=Get Is Item View())
   > n3 = Switch on E_QuestPunchCardState(Selection=Get CurrentState() -> (CurrentState=NewEnumerator0))
   > SetFocus(self=Get ListView())
   n4 = Get CurrentSelectedItem()
   SetFocus(
    execute=@n3.NewEnumerator1
    self=Select(Option 0:Widget Object Reference=Get FocusCatcher(), Option 1:Widget Object Reference=Get CurrentSelectedItem(), Index=IsValid(Object=@n4)) -> (ReturnValue:Widget Object Reference)
   )
   SetFocus(execute=@n3.NewEnumerator2, self=Get FocusCatcher())
   SetFocus(execute=@n2.else, self=Get FocusCatcher())
  ]}
 UpdateHintVisibility {nodes [
   UpdateHintVisibility() {is=title K2Node_FunctionEntry}
   n1 = IsGamepad() {is=title K2Node_CallFunction}
   n2 = "Not Equal (Enum)"(A=Get CurrentState() -> (CurrentState=NewEnumerator0), B=NewEnumerator0)
   > SetVisibility(
    self=Get Buttons()
    InVisibility=Select(
     Option 0:ESlateVisibility Enum=SelfHitTestInvisible
     Option 1:ESlateVisibility Enum=Collapsed
     Index=OR Boolean!(A=@n1, B=@n2) -> (ReturnValue)
    ) -> (ReturnValue:ESlateVisibility Enum)
   )
  ]}
 UpdateNextItemProgressStyle {nodes [
   UpdateNextItemProgressStyle() {is=title K2Node_FunctionEntry}
   n1 = GetScrollOffset!(self:List View Base Object Reference=Get ListView()) -> (ReturnValue:float)
   n2 = Get NormalScrollOffset() -> (NormalScrollOffset:double)
   > n3 = Cast To QuestUIGoalRewardVM(Object=GetItemAt(self=Get ListView(), Index=float + float(A=@n1, B=@n2) -> (ReturnValue:int)))
   n4 = Get ProgressValue!(self:Quest UIGoal Reward VM Object Reference=@n3) -> (ProgressValue:double)
   > UpdateLeftProgressState!(
    execute
    self:WBP Goal Reward Object Reference=Get finalGoalReward()
    ProgressValue:float=Select(Option 0:float=float * float(A=@n4, B=2) {is=title K2Node_PromotableOperator}, Option 1:float=1, Index="float >= float"(A=@n4, B=0.5)) -> (ReturnValue:float)
    IsDone:bool="Equal (Float)"(A=@n4, B=1)
   ) -> (then)
  ]}
 UpdateObjectiveInfo {nodes [
   n1 = UpdateObjectiveInfo!() -> (then, QuestVM:Quest UI Quest View Model Object Reference, CanShowOverallProgress:bool=~) {is=title K2Node_FunctionEntry}
   n2 = Get QuestUIHorizontalPunchCardVM()
   > SetVisibility(
    self=Get Image_Icon_CQ() -> (Image_Icon_CQ:Image Object Reference)
    InVisibility=Select(
     Option 0:ESlateVisibility Enum=Collapsed
     Option 1:ESlateVisibility Enum=HitTestInvisible
     Index=Get bDisplayCommunityBadge!(self:Quest UI Punch Card View Model Object Reference=@n2) -> (bDisplayCommunityBadge:bool)
    ) -> (ReturnValue:ESlateVisibility Enum)
   )
   n3 = Get QuestUIHorizontalPunchCardVM()
   > SetVisibility(
    self=Get VerticalBox() -> (VerticalBox:Vertical Box Object Reference)
    InVisibility=Select(
     Option 0=Collapsed
     Option 1=HitTestInvisible
     Index=Get bDisplayExpirationTime!(self:Quest UI Punch Card View Model Object Reference=@n3) -> (bDisplayExpirationTime:bool)
    )
   )
   > n4 = Is Valid(InputObject=@n1)
   n5 = Branch(execute=@n4.Is Not Valid, Condition=Get CanShowOverallProgress() -> (CanShowOverallProgress:bool))
   n6 = Get QuestUIHorizontalPunchCardVM()
   n7 = GetAchievedCount!(self:Quest UI Punch Card View Model Object Reference=@n6) -> (OutAchievedCount:int=0, OutRequiredCount:int=0)
   n8 = Format Text!(
    Format=Get SubHeaderText!(self:Quest UI Punch Card View Model Object Reference=@n6) -> (SubHeaderText:Text)
    0:int=@n7
    1:int=@n7.OutRequiredCount
   ) -> (Result)
   > "SetText (Text)"(self=Get ObjectiveInfo() -> (ObjectiveInfo:Common Text Object Reference), InText=@n8)
   > SetText(self=Get Rich_ObjectiveQuestText() -> (Rich_ObjectiveQuestText:Rich Text Block Object Reference), InText=@n8)
   "SetText (Text)"(execute=@n5.else, self=Get ObjectiveInfo())
   > SetText(self=Get Rich_ObjectiveQuestText())
   n9 = Get OrderedContentVMs(self=Get QuestUIHorizontalPunchCardVM())
   n10 = Cast To QuestUIQuestVM!(Object="Get (a copy)"(
     Array:Array of Quest UI Base Content View Model Object References=@n9
     Dimension 1=Last Index(TargetArray:Array of Quest UI Base Content View Model Object References=@n9) -> (ReturnValue=0)
    ) -> (Output:Quest UI Base Content View Model Object Reference)) -> (AsQuest UI Quest View Model, bSuccess:bool=~)
   n11 = Get QuestVM!() -> (QuestVM)
   n12 = Get QuestVM()
   n13 = Select!(
    NotDisplay:Text=Get QuestName!(self:Quest UI Quest View Model Object Reference=Get QuestVM()) -> (QuestName:Text)
    DisplayLastStage:Text=Format Text!(
     Format="LOCTEXT(\"{body}: {Progress}/{Total}\")"
     body:Text=Get QuestName(self=@n11)
     Progress:int=Get AchievedCount(self=@n10) -> (AchievedCount=0)
     Total:int=Get RequiredCount(self=@n10) -> (RequiredCount=0)
    ) -> (Result)
    DisplayCurrentStage:Text=Format Text(
     Format="LOCTEXT(\"{body}: {Progress}/{Total}\")"
     body=Get QuestName(self=@n12)
     Progress=Get AchievedCount(self=@n12) -> (AchievedCount=0)
     Total=Get RequiredCount(self=@n12) -> (RequiredCount=0)
    )
    Index:EGoalCardProgressDisplayMode Enum=Get ProgressDisplayMode!(self:Quest UI Punch Card View Model Object Reference=Get QuestUIHorizontalPunchCardVM()) -> (ProgressDisplayMode:EGoalCardProgressDisplayMode Enum=NotDisplay)
   ) -> (ReturnValue:Text)
   "SetText (Text)"(execute=@n4.Is Valid, self=Get ObjectiveInfo(), InText=@n13)
   > SetText(self=Get Rich_ObjectiveQuestText(), InText=@n13)
  ]}
 UpdateOutlineVisibility {nodes [
   UpdateOutlineVisibility() {is=title K2Node_FunctionEntry}
   n1 = NOT Boolean(A="Equal (Enum)"(A:E_QuestPunchCardState Enum=Get CurrentState() -> (CurrentState=NewEnumerator0), B:E_QuestPunchCardState Enum=NewEnumerator2))
   n2 = Get Is Item View()
   > SetVisibility(
    self=Get OutlineOverlay() -> (OutlineOverlay:Overlay Object Reference)
    InVisibility=Select!(
     Option 0:ESlateVisibility Enum=SelfHitTestInvisible
     Option 1:ESlateVisibility Enum=Hidden
     Index:bool=AND Boolean!(A=@n2, B=@n1) -> (ReturnValue)
    ) -> (ReturnValue:ESlateVisibility Enum)
   )
  ]}
 UpdateProgress {nodes [
   n1 = UpdateProgress!() -> (then, Progress:int) {is=title K2Node_FunctionEntry}
   > Branch(Condition="integer > integer"(A=@n1))
   > Refresh() {is=title K2Node_CallFunction}
  ]}
]
variables [
 As Quest UIGoal Reward VM {type=QuestUIGoalRewardVM_C, default=null}
 BlockTiming {
  type=UIKitBlockTiming
  default {
   deselectingFocused=0.1
   deselectingIdle=0.1
   disabling=0
   enabling=0
   focusing=0
   hovering=0
   pressing=0
   releasing=0
   selecting=0.1
   unfocusing=0
   unhovering=0
  }
 }
 CurrentSelectedItem {type=CommonButtonBase, default=null}
 CurrentState {type=E_QuestPunchCardState, default=HasPages}
 DynamicEntryBoxPaddings {default {2=-252, 3=-85, 4=-29, 5=0}}
 GoalRewardVMs {type="array<QuestUIGoalRewardVM_C>", default []}
 Is Item View {type=boolean}
 ItemCountScrollThreshold {type=integer, default=5}
 NormalScrollOffset {type=number, default=4}
 OnGoalsDoubleClick {}
 OnGoalsNavigate {}
 OnRewardViewChanged {}
 OnSelectedIndexChanged {}
 RewardVMs {type="array<QuestUIRewardVM>", default []}
 SelectedIndex {type=integer, default=-1}
]
widgets {
 info {
  inheritedWidgetCount=1
  parentClass=/QuestUI/QuestScreen/QuestTile/WBP_QuestTile_Base.WBP_QuestTile_Base_C
  rootWidgetClass=&SizeBox
  widgetCount=34
 }
 widgets [
  SizeBox_0 {
   widgetClassPath=&SizeBox
   widget=SizeBox_0
   properties {bOverride_WidthOverride=true, heightOverride=410, minDesiredHeight=320, widthOverride=872}
  }
  Root {widgetClassPath=&GridPanel, parent=SizeBox_0, widget=Root, properties {columnFill [1], rowFill [0,1]}, slot {class=SizeBoxSlot}}
  StatusIndicator {
   widgetClassPath=/FortUILibrary/UIKit/StatusIndicator/WBP_UIKit_StatusIndicator
   bIsVariable=true
   parent=Root
   widget=StatusIndicator
   properties {type=Attention, configuration=Dot, icon=/FortUILibrary/Textures/Icons/T_UI_Icon_UIKit_Check, iconSize=Compact, visibility=Collapsed}
   slot {class=GridSlot, properties {horizontalAlignment=HAlign_Right, layer=10, verticalAlignment=VAlign_Top, padding=<ltrb 0 8 8 0>}}
  }
  CommonVisualAttachment_0 {
   widgetClassPath=&CommonVisualAttachment
   parent=Root
   widget=CommonVisualAttachment_0
   slot {class=GridSlot, properties {horizontalAlignment=HAlign_Left, layer=1, verticalAlignment=VAlign_Top, padding=<ltrb 41 0 0 0>}}
  }
  DebugBundleID {
   widgetClassPath=&CommonTextBlock
   bIsVariable=true
   parent=CommonVisualAttachment_0
   widget=DebugBundleID
   properties {
    bAutoCollapseWithEmptyText=true
    text=Debug Bundle ID
    colorAndOpacity {colorUseRule=UseColor_Specified, specifiedColor=<rgba 0.5776 0.5776 0.5776 1>}
    font {
     bForceMonospaced=false
     bMaterialIsStencil=false
     fontMaterial=null
     fontObject=&HeadingNow
     letterSpacing=0
     monospacedWidth=1
     size=10.5
     skewAmount=0
     typefaceFontName=74 Regular
     outlineSettings {
      bApplyOutlineToDropShadows=false
      bMiteredCorners=false
      bSeparateFillAlpha=false
      outlineMaterial=null
      outlineSize=0
      outlineColor=<rgba 0 0 0 1>
     }
    }
    renderTransform {angle=0, scale=<1 1>, shear=<0 0>, translation=<0 -17>}
    shadowOffset=<0 0>
   }
   slot {class=SizeBoxSlot}
  }
  Subheader {
   widgetClassPath=&HorizontalBox
   bIsVariable=true
   parent=Root
   widget=Subheader
   slot {class=GridSlot, properties {layer=1, verticalAlignment=VAlign_Top, padding=<ltrb 40 10 24 0>}}
  }
  Image_Icon_CQ {
   widgetClassPath=&Image
   bIsVariable=true
   parent=Subheader
   widget=Image_Icon_CQ
   properties {
    visibility=Collapsed
    brush {
     bIsDynamicallyLoaded=false
     drawAs=Image
     imageType=FullColor
     mirroring=NoMirror
     resourceName=None
     resourceObject=/Game/Athena/UI/Challenges/Art/T_UI_CommunityQuest_64
     tiling=NoTile
     imageSize=<18 18>
     margin=<ltrb 0 0 0 0>
     outlineSettings {
      bUseBrushTransparency=false
      roundingType=HalfHeightRadius
      width=0
      color {colorUseRule=UseColor_Specified, specifiedColor=<rgba 0 0 0 0>}
      cornerRadii=<0 0 0 0>
     }
     tintColor {colorUseRule=UseColor_Specified, specifiedColor=<rgba 1 1 1 1>}
     uVRegion {bIsValid=false, max=<0 0>, min=<0 0>}
    }
    renderTransform {angle=0, scale=<2 2>, shear=<0 0>, translation=<-4 0>}
   }
   slot {class=HorizontalBoxSlot, properties {horizontalAlignment=HAlign_Left, verticalAlignment=VAlign_Top, padding=<ltrb 10 5 9 4>}}
  }
  Switch_ObjectiveInfo {
   widgetClassPath=&CommonVisibilitySwitcher
   parent=Subheader
   widget=Switch_ObjectiveInfo
   properties {activeWidgetIndex=1}
   slot {class=HorizontalBoxSlot, properties {size {sizeRule=Fill, value=1}}}
  }
  ObjectiveInfo {
   widgetClassPath=&CommonTextBlock
   bIsVariable=true
   parent=Switch_ObjectiveInfo
   widget=ObjectiveInfo
   properties {
    autoWrapText=true
    bIsScrollingEnabled=false
    lineHeightPercentage=1.15
    scrollOrientation=Orient_Vertical
    scrollStyle=&TextScrollStyle-Base_C
    text=Complete Sample Quest
    colorAndOpacity {colorUseRule=UseColor_Specified, specifiedColor=<rgba 0.491 0.4793 0.5972 1>}
    font {
     bForceMonospaced=false
     bMaterialIsStencil=false
     fontMaterial=null
     fontObject=&HeadingNow
     letterSpacing=0
     monospacedWidth=1
     size=15.75
     skewAmount=0
     typefaceFontName=75 Medium
     outlineSettings {
      bApplyOutlineToDropShadows=false
      bMiteredCorners=false
      bSeparateFillAlpha=false
      outlineMaterial=null
      outlineSize=0
      outlineColor=<rgba 0 0 0 1>
     }
    }
    shadowOffset=<0 0>
   }
   slot {class=CommonVisibilitySwitcherSlot}
  }
  Rich_ObjectiveQuestText {
   widgetClassPath=/Script/UMG.RichTextBlock
   bIsVariable=true
   parent=Switch_ObjectiveInfo
   widget=Rich_ObjectiveQuestText
   properties {
    autoWrapText=true
    text=TextBlock_Rich
    textStyleSet=/Game/Athena/HUD/QuestIcons/DT_RichText_QuestTile
    decoratorClasses [/Game/Athena/HUD/QuestIcons/BP_QuestTile_IconDecorator.BP_QuestTile_IconDecorator_C]
   }
   slot {class=CommonVisibilitySwitcherSlot}
  }
  ListSwitcher {
   widgetClassPath=&CommonVisibilitySwitcher
   bIsVariable=true
   parent=Root
   widget=ListSwitcher
   slot {class=GridSlot, properties {layer=5, row=1, verticalAlignment=VAlign_Top, padding=<ltrb 16 0 16 0>}}
  }
  ListViewRoot {
   widgetClassPath=&GridPanel
   bIsVariable=true
   parent=ListSwitcher
   widget=ListViewRoot
   properties {columnFill [1], rowFill [1]}
   slot {class=CommonVisibilitySwitcherSlot}
  }
  ListView {
   widgetClassPath=&CommonListView
   bIsVariable=true
   parent=ListViewRoot
   widget=ListView
   properties {
    bEnableRightClickScrolling=false
    bEnableTouchScrolling=false
    bReturnFocusToSelection=true
    consumeMouseWheel=Never
    entryWidgetClass=&/WBP_GoalReward.WBP_GoalReward_C
    navigation=&self:WidgetTree.ListView.WidgetNavigation_0
    numDesignerPreviewEntries=9
    orientation=Orient_Horizontal
    scrollIntoViewAlignment=TopOrLeft
    visibility=SelfHitTestInvisible
    wheelScrollMultiplier=0
    scrollBarStyle {
     thickness=0
     draggedThumbImage {
      bIsDynamicallyLoaded=false
      drawAs=NoDrawType
      imageType=FullColor
      mirroring=NoMirror
      resourceName=../../../Engine/Content/Slate/Common/Scrollbar_Thumb.png
      resourceObject=null
      tiling=NoTile
      imageSize=<0 0>
      margin=<ltrb 0.25 0.25 0.25 0.25>
      outlineSettings {
       bUseBrushTransparency=false
       roundingType=HalfHeightRadius
       width=0
       color {colorUseRule=UseColor_Specified, specifiedColor=<rgba 0 0 0 0>}
       cornerRadii=<0 0 0 0>
      }
      tintColor {colorUseRule=UseColor_Specified, specifiedColor=<rgba 1 1 1 1>}
      uVRegion {bIsValid=false, max=<0 0>, min=<0 0>}
     }
     horizontalBackgroundImage {
      bIsDynamicallyLoaded=false
      drawAs=NoDrawType
      imageType=FullColor
      mirroring=NoMirror
      resourceName=None
      resourceObject=null
      tiling=NoTile
      imageSize=<256 256>
      margin=<ltrb 0 0 0 0>
      outlineSettings {
       bUseBrushTransparency=false
       roundingType=HalfHeightRadius
       width=0
       color {colorUseRule=UseColor_Specified, specifiedColor=<rgba 0 0 0 0>}
       cornerRadii=<0 0 0 0>
      }
      tintColor {colorUseRule=UseColor_Specified, specifiedColor=<rgba 1 1 1 1>}
      uVRegion {bIsValid=false, max=<0 0>, min=<0 0>}
     }
     horizontalBottomSlotImage {
      bIsDynamicallyLoaded=false
      drawAs=NoDrawType
      imageType=FullColor
      mirroring=NoMirror
      resourceName=../../../Engine/Content/Slate/Common/Scrollbar_Background_Horizontal.png
      resourceObject=null
      tiling=NoTile
      imageSize=<8 8>
      margin=<ltrb 0 0 0 0>
      outlineSettings {
       bUseBrushTransparency=false
       roundingType=HalfHeightRadius
       width=0
       color {colorUseRule=UseColor_Specified, specifiedColor=<rgba 0 0 0 0>}
       cornerRadii=<0 0 0 0>
      }
      tintColor {colorUseRule=UseColor_Specified, specifiedColor=<rgba 1 1 1 1>}
      uVRegion {bIsValid=false, max=<0 0>, min=<0 0>}
     }
     horizontalTopSlotImage {
      bIsDynamicallyLoaded=false
      drawAs=NoDrawType
      imageType=FullColor
      mirroring=NoMirror
      resourceName=../../../Engine/Content/Slate/Common/Scrollbar_Background_Horizontal.png
      resourceObject=null
      tiling=NoTile
      imageSize=<8 8>
      margin=<ltrb 0 0 0 0>
      outlineSettings {
       bUseBrushTransparency=false
       roundingType=HalfHeightRadius
       width=0
       color {colorUseRule=UseColor_Specified, specifiedColor=<rgba 0 0 0 0>}
       cornerRadii=<0 0 0 0>
      }
      tintColor {colorUseRule=UseColor_Specified, specifiedColor=<rgba 1 1 1 1>}
      uVRegion {bIsValid=false, max=<0 0>, min=<0 0>}
     }
     hoveredThumbImage {
      bIsDynamicallyLoaded=false
      drawAs=NoDrawType
      imageType=FullColor
      mirroring=NoMirror
      resourceName=../../../Engine/Content/Slate/Common/Scrollbar_Thumb.png
      resourceObject=null
      tiling=NoTile
      imageSize=<0 0>
      margin=<ltrb 0.25 0.25 0.25 0.25>
      outlineSettings {
       bUseBrushTransparency=false
       roundingType=HalfHeightRadius
       width=0
       color {colorUseRule=UseColor_Specified, specifiedColor=<rgba 0 0 0 0>}
       cornerRadii=<0 0 0 0>
      }
      tintColor {colorUseRule=UseColor_Specified, specifiedColor=<rgba 1 1 1 1>}
      uVRegion {bIsValid=false, max=<0 0>, min=<0 0>}
     }
     normalThumbImage {
      bIsDynamicallyLoaded=false
      drawAs=NoDrawType
      imageType=FullColor
      mirroring=NoMirror
      resourceName=../../../Engine/Content/Slate/Common/Scrollbar_Thumb.png
      resourceObject=null
      tiling=NoTile
      imageSize=<0 0>
      margin=<ltrb 0.25 0.25 0.25 0.25>
      outlineSettings {
       bUseBrushTransparency=false
       roundingType=HalfHeightRadius
       width=0
       color {colorUseRule=UseColor_Specified, specifiedColor=<rgba 0 0 0 0>}
       cornerRadii=<0 0 0 0>
      }
      tintColor {colorUseRule=UseColor_Specified, specifiedColor=<rgba 1 1 1 1>}
      uVRegion {bIsValid=false, max=<0 0>, min=<0 0>}
     }
     verticalBackgroundImage {
      bIsDynamicallyLoaded=false
      drawAs=NoDrawType
      imageType=NoImage
      mirroring=NoMirror
      resourceName=None
      resourceObject=null
      tiling=NoTile
      imageSize=<0 0>
      margin=<ltrb 0 0 0 0>
      outlineSettings {
       bUseBrushTransparency=false
       roundingType=HalfHeightRadius
       width=0
       color {colorUseRule=UseColor_Specified, specifiedColor=<rgba 0 0 0 0>}
       cornerRadii=<0 0 0 0>
      }
      tintColor {colorUseRule=UseColor_Specified, specifiedColor=<rgba 1 1 1 1>}
      uVRegion {bIsValid=false, max=<0 0>, min=<0 0>}
     }
     verticalBottomSlotImage {
      bIsDynamicallyLoaded=false
      drawAs=NoDrawType
      imageType=FullColor
      mirroring=NoMirror
      resourceName=../../../Engine/Content/Slate/Common/Scrollbar_Background_Vertical.png
      resourceObject=null
      tiling=NoTile
      imageSize=<8 8>
      margin=<ltrb 0 0 0 0>
      outlineSettings {
       bUseBrushTransparency=false
       roundingType=HalfHeightRadius
       width=0
       color {colorUseRule=UseColor_Specified, specifiedColor=<rgba 0 0 0 0>}
       cornerRadii=<0 0 0 0>
      }
      tintColor {colorUseRule=UseColor_Specified, specifiedColor=<rgba 1 1 1 1>}
      uVRegion {bIsValid=false, max=<0 0>, min=<0 0>}
     }
     verticalTopSlotImage {
      bIsDynamicallyLoaded=false
      drawAs=NoDrawType
      imageType=FullColor
      mirroring=NoMirror
      resourceName=None
      resourceObject=/Game/Environments/Asteria/Sets/EclipseHotel/Textures/Numbers/102
      tiling=NoTile
      imageSize=<256 256>
      margin=<ltrb 0 0 0 0>
      outlineSettings {
       bUseBrushTransparency=false
       roundingType=HalfHeightRadius
       width=0
       color {colorUseRule=UseColor_Specified, specifiedColor=<rgba 0 0 0 0>}
       cornerRadii=<0 0 0 0>
      }
      tintColor {colorUseRule=UseColor_Specified, specifiedColor=<rgba 1 1 1 1>}
      uVRegion {bIsValid=false, max=<0 0>, min=<0 0>}
     }
    }
   }
   slot {class=GridSlot}
  }
  finalGoalReward {
   widgetClassPath=&/WBP_GoalReward
   bIsVariable=true
   parent=ListViewRoot
   widget=finalGoalReward
   properties {
    holdData=&CommonUI_HoldData_Base_C
    navigation=&self:WidgetTree.finalGoalReward.WidgetNavigation_0
   }
   slot {class=GridSlot, properties {column=1}}
  }
  ListViewProgressPatch {
   widgetClassPath=&Image
   bIsVariable=true
   parent=ListViewRoot
   widget=ListViewProgressPatch
   properties {brush {
     bIsDynamicallyLoaded=false
     drawAs=Image
     imageType=FullColor
     mirroring=NoMirror
     resourceName=None
     resourceObject=/FortniteCoreUI/UI/Quests/MaterialInstances/MI_UI_QuestPunchCardProgress_Background
     tiling=NoTile
     imageSize=<27 40>
     margin=<ltrb 0 0 0 0>
     outlineSettings {
      bUseBrushTransparency=false
      roundingType=HalfHeightRadius
      width=0
      color {colorUseRule=UseColor_Specified, specifiedColor=<rgba 0 0 0 0>}
      cornerRadii=<0 0 0 0>
     }
     tintColor {colorUseRule=UseColor_Specified, specifiedColor=<rgba 1 1 1 1>}
     uVRegion {bIsValid=false, max=<0 0>, min=<0 0>}
    }}
   slot {class=GridSlot, properties {horizontalAlignment=HAlign_Left, layer=5, verticalAlignment=VAlign_Bottom, padding=<ltrb -2 0 0 0>}}
  }
  Spacer_483 {
   widgetClassPath=&Spacer
   parent=ListViewRoot
   widget=Spacer_483
   properties {size=<1 180>}
   slot {class=GridSlot, properties {horizontalAlignment=HAlign_Left}}
  }
  DynamicEntryBox_Fill {
   widgetClassPath=&DynamicEntryBox
   bIsVariable=true
   parent=ListSwitcher
   widget=DynamicEntryBox_Fill
   properties {
    entryWidgetClass=&WBP_GoalReward_Container_C
    numDesignerPreviewEntries=5
    entrySizeRule {sizeRule=Fill, value=1}
   }
   slot {class=CommonVisibilitySwitcherSlot}
  }
  WBP_GoalReward_Solo {
   widgetClassPath=&/WBP_GoalReward_Solo
   bIsVariable=true
   parent=ListSwitcher
   widget=WBP_GoalReward_Solo
   slot {class=CommonVisibilitySwitcherSlot, properties {padding=<ltrb 24 0 25 0>}}
  }
  Buttons {
   widgetClassPath=&GridPanel
   bIsVariable=true
   parent=Root
   widget=Buttons
   properties {visibility=Collapsed, columnFill [0,1,0]}
   slot {class=GridSlot, properties {horizontalAlignment=HAlign_Center, layer=10, row=2, verticalAlignment=VAlign_Top, padding=<ltrb 0 2 0 0>}}
  }
  GoalsHorizontal_PreviousButton {
   widgetClassPath=&WBP_UIKit_Button_Generic
   bIsVariable=true
   parent=Buttons
   widget=GoalsHorizontal_PreviousButton
   properties {
    bHideInputActionWithKeyboard=true
    bIsFocusable=false
    bShouldUseFallbackDefaultInputAction=false
    backgroundMaterial=/FortUILibrary/UIKit/Buttons/MaterialInstances/CTAButtons/MI_UIKit_BgRegular
    enableFocusedSelectedState=false
    iconMaterial=/FortUILibrary/Foundation/Buttons/Materials/Arrows/MI_UIKit_Icon_Button_Previous
    minHeight=38
    minWidth=38
    mobileMinHeight=80
    mobileMinWidth=80
    promptHorizontalAlignment=HAlign_Center
    useTextBlock=false
    buttonPadding=<ltrb 0 0 0 0>
    iconPadding=<ltrb 4 4 4 4>
    iconSize=<16 16>
    promptNudge=<0 0>
   }
   slot {class=GridSlot, properties {verticalAlignment=VAlign_Center, padding=<ltrb 0 0 15 0>}}
  }
  GoalsHorizontal_NextButton {
   widgetClassPath=&WBP_UIKit_Button_Generic
   bIsVariable=true
   parent=Buttons
   widget=GoalsHorizontal_NextButton
   properties {
    bHideInputActionWithKeyboard=true
    bIsFocusable=false
    bShouldUseFallbackDefaultInputAction=false
    backgroundMaterial=&MI_UIKit_BgRegular
    enableFocusedSelectedState=false
    iconMaterial=/FortUILibrary/Foundation/Buttons/Materials/Arrows/MI_UIKit_Icon_Button_Next
    minHeight=38
    minWidth=38
    mobileMinHeight=80
    mobileMinWidth=80
    promptHorizontalAlignment=HAlign_Center
    useTextBlock=false
    buttonPadding=<ltrb 0 0 0 0>
    iconPadding=<ltrb 4 4 4 4>
    iconSize=<16 16>
    promptNudge=<0 0>
   }
   slot {class=GridSlot, properties {column=2, layer=10, verticalAlignment=VAlign_Center, padding=<ltrb 15 0 0 0>}}
  }
  OutlineOverlay {
   widgetClassPath=&Overlay
   bIsVariable=true
   parent=Root
   widget=OutlineOverlay
   slot {class=GridSlot, properties {layer=6, rowSpan=4}}
  }
  Block_Outline {
   widgetClassPath=/FortUILibrary/UIKit/Buttons/Blocks/WBP_UIKit_Block_Outline
   bIsVariable=true
   parent=OutlineOverlay
   widget=Block_Outline
   properties {bottomLeftCornerRadius=20, bottomRightCornerRadius=20, radiusInPixels=true, topLeftCornerRadius=20, topRightCornerRadius=20}
   slot {class=OverlaySlot, properties {horizontalAlignment=HAlign_Fill, verticalAlignment=VAlign_Fill}}
  }
  SelectedStateOutline {
   widgetClassPath=&Image
   bIsVariable=true
   parent=OutlineOverlay
   widget=SelectedStateOutline
   properties {
    visibility=Collapsed
    brush {
     bIsDynamicallyLoaded=false
     drawAs=Image
     imageType=FullColor
     mirroring=NoMirror
     resourceName=None
     resourceObject=/QuestUI/QuestScreen/QuestTile/Materials/MI_UI_Outline_Selected
     tiling=NoTile
     imageSize=<0 0>
     margin=<ltrb 0 0 0 0>
     outlineSettings {
      bUseBrushTransparency=false
      roundingType=HalfHeightRadius
      width=0
      color {colorUseRule=UseColor_Specified, specifiedColor=<rgba 0 0 0 0>}
      cornerRadii=<0 0 0 0>
     }
     tintColor {colorUseRule=UseColor_Specified, specifiedColor=<rgba 1 1 1 1>}
     uVRegion {bIsValid=false, max=<0 0>, min=<0 0>}
    }
   }
   slot {class=OverlaySlot, properties {horizontalAlignment=HAlign_Fill, verticalAlignment=VAlign_Fill}}
  }
  FocusCatcher {
   widgetClassPath=&Button
   parent=Root
   widget=FocusCatcher
   properties {
    renderOpacity=0
    widgetStyle {
     clickedSlateSound {resourceObject=null}
     disabled {
      bIsDynamicallyLoaded=false
      drawAs=NoDrawType
      imageType=NoImage
      mirroring=NoMirror
      resourceName=None
      resourceObject=null
      tiling=NoTile
      imageSize=<0 0>
      margin=<ltrb 0 0 0 0>
      outlineSettings {
       bUseBrushTransparency=false
       roundingType=HalfHeightRadius
       width=0
       color {colorUseRule=UseColor_Specified, specifiedColor=<rgba 0 0 0 0>}
       cornerRadii=<0 0 0 0>
      }
      tintColor {colorUseRule=UseColor_Specified, specifiedColor=<rgba 1 1 1 1>}
      uVRegion {bIsValid=false, max=<0 0>, min=<0 0>}
     }
     disabledForeground {colorUseRule=UseColor_Specified, specifiedColor=<rgba 0.5271 0.5271 0.5271 1>}
     hovered {
      bIsDynamicallyLoaded=false
      drawAs=RoundedBox
      imageType=NoImage
      mirroring=NoMirror
      resourceName=None
      resourceObject=null
      tiling=NoTile
      imageSize=<32 1>
      margin=<ltrb 0 0 0 0>
      outlineSettings {
       bUseBrushTransparency=true
       roundingType=FixedRadius
       width=1
       color {colorUseRule=UseColor_Specified, specifiedColor=<rgba 0.7243 0.7243 0.7243 1>}
       cornerRadii=<4 4 4 4>
      }
      tintColor {colorUseRule=UseColor_Specified, specifiedColor=<rgba 0.7243 0.7243 0.7243 1>}
      uVRegion {bIsValid=false, max=<0 0>, min=<0 0>}
     }
     hoveredForeground {colorUseRule=UseColor_Specified, specifiedColor=<rgba 0 1 1 1>}
     hoveredSlateSound {resourceObject=null}
     normal {
      bIsDynamicallyLoaded=false
      drawAs=RoundedBox
      imageType=NoImage
      mirroring=NoMirror
      resourceName=None
      resourceObject=null
      tiling=NoTile
      imageSize=<32 32>
      margin=<ltrb 0 0 0 0>
      outlineSettings {
       bUseBrushTransparency=true
       roundingType=FixedRadius
       width=1
       color {colorUseRule=UseColor_Specified, specifiedColor=<rgba 0.6951 0.6951 0.6951 1>}
       cornerRadii=<4 4 4 4>
      }
      tintColor {colorUseRule=UseColor_Specified, specifiedColor=<rgba 0.4955 0.4955 0.4955 1>}
      uVRegion {bIsValid=false, max=<0 0>, min=<0 0>}
     }
     normalForeground {colorUseRule=UseColor_Specified, specifiedColor=<rgba 0 1 1 1>}
     normalPadding=<ltrb 12 1.5 12 1.5>
     pressed {
      bIsDynamicallyLoaded=false
      drawAs=RoundedBox
      imageType=NoImage
      mirroring=NoMirror
      resourceName=None
      resourceObject=null
      tiling=NoTile
      imageSize=<32 1>
      margin=<ltrb 0 0 0 0>
      outlineSettings {
       bUseBrushTransparency=true
       roundingType=FixedRadius
       width=1
       color {colorUseRule=UseColor_Specified, specifiedColor=<rgba 0.7243 0.7243 0.7243 1>}
       cornerRadii=<4 4 4 4>
      }
      tintColor {colorUseRule=UseColor_Specified, specifiedColor=<rgba 0.3843 0.3843 0.3843 1>}
      uVRegion {bIsValid=false, max=<0 0>, min=<0 0>}
     }
     pressedForeground {colorUseRule=UseColor_Specified, specifiedColor=<rgba 0 1 1 1>}
     pressedPadding=<ltrb 12 1.5 12 1.5>
     pressedSlateSound {resourceObject=null}
    }
   }
   slot {class=GridSlot, properties {layer=100}}
  }
  Background {
   widgetClassPath=&Image
   bIsVariable=true
   parent=Root
   widget=Background
   properties {brush {
     bIsDynamicallyLoaded=false
     drawAs=Image
     imageType=FullColor
     mirroring=NoMirror
     resourceName=None
     resourceObject=/FortniteCoreUI/UI/Quests/MaterialInstances/MI_UI_QuestBackplate
     tiling=NoTile
     imageSize=<0 0>
     margin=<ltrb 0 0 0 0>
     outlineSettings {
      bUseBrushTransparency=false
      roundingType=HalfHeightRadius
      width=0
      color {colorUseRule=UseColor_Specified, specifiedColor=<rgba 0 0 0 0>}
      cornerRadii=<0 0 0 0>
     }
     tintColor {colorUseRule=UseColor_Specified, specifiedColor=<rgba 1 1 1 1>}
     uVRegion {bIsValid=false, max=<0 0>, min=<0 0>}
    }}
   slot {class=GridSlot, properties {rowSpan=4}}
  }
  Spacer_99 {widgetClassPath=&Spacer, parent=Root, widget=Spacer_99, properties {size=<1 10>}, slot {class=GridSlot, properties {row=3}}}
  CommonVisualAttachment_280 {
   widgetClassPath=&CommonVisualAttachment
   parent=Root
   widget=CommonVisualAttachment_280
   properties {contentAnchor=<1 0.5>}
   slot {class=GridSlot, properties {horizontalAlignment=HAlign_Right, layer=10, verticalAlignment=VAlign_Top, padding=<ltrb 0 0 12 0>}}
  }
  VerticalBox {
   widgetClassPath=&VerticalBox
   bIsVariable=true
   parent=CommonVisualAttachment_280
   widget=VerticalBox
   slot {class=SizeBoxSlot, properties {padding=<ltrb 10 0 0 0>}}
  }
  Overlay_0 {
   widgetClassPath=&Overlay
   parent=VerticalBox
   widget=Overlay_0
   slot {class=VerticalBoxSlot, properties {horizontalAlignment=HAlign_Right, verticalAlignment=VAlign_Top}}
  }
  TimerBack {
   widgetClassPath=&Image
   bIsVariable=true
   parent=Overlay_0
   widget=TimerBack
   properties {
    brush {
     bIsDynamicallyLoaded=false
     drawAs=RoundedBox
     imageType=NoImage
     mirroring=NoMirror
     resourceName=None
     resourceObject=null
     tiling=NoTile
     imageSize=<32 32>
     margin=<ltrb 0 0 0 0>
     outlineSettings {
      bUseBrushTransparency=false
      roundingType=HalfHeightRadius
      width=0
      color {colorUseRule=UseColor_Specified, specifiedColor=<rgba 0 0 0 0>}
      cornerRadii=<0 0 0 0>
     }
     tintColor {colorUseRule=UseColor_Specified, specifiedColor=<rgba 1 1 1 1>}
     uVRegion {bIsValid=false, max=<0 0>, min=<0 0>}
    }
    colorAndOpacity=<rgba 0.0048 0.004 0.0116 1>
   }
   slot {class=OverlaySlot, properties {horizontalAlignment=HAlign_Fill, verticalAlignment=VAlign_Fill}}
  }
  ExpirationTime {
   widgetClassPath=/Script/FortniteUI.AthenaChallengeDateTimeTextBlock
   bIsVariable=true
   parent=Overlay_0
   widget=ExpirationTime
   properties {
    bAllowSeconds=false
    bAutoCollapseWithEmptyText=true
    bIsScrollingEnabled=false
    completionText=EXPIRED
    labelFormat="Ends: {0}"
    text=Text Block
    font {
     bForceMonospaced=false
     bMaterialIsStencil=false
     fontMaterial=null
     fontObject=&HeadingNow
     letterSpacing=0
     monospacedWidth=1
     size=14.25
     skewAmount=0
     typefaceFontName=74 Regular
     outlineSettings {
      bApplyOutlineToDropShadows=false
      bMiteredCorners=false
      bSeparateFillAlpha=false
      outlineMaterial=null
      outlineSize=0
      outlineColor=<rgba 0 0 0 1>
     }
    }
    shadowOffset=<0 0>
   }
   slot {class=OverlaySlot, properties {horizontalAlignment=HAlign_Center, verticalAlignment=VAlign_Center, padding=<ltrb 10 0 10 0>}}
  }
  DecoyButtonBack {
   widgetClassPath=/Game/UI/Frontend/BottomBar/WBP_BottomBarDecoyButton
   bIsVariable=true
   parent=Root
   widget=DecoyButtonBack
   properties {priority=10, triggeringInputAction {dataTable=/QuestUI/QuestScreen/RewardsPanel/QuestInputActionDataTable, rowName=BackToBundleView}}
   slot {class=GridSlot, properties {horizontalAlignment=HAlign_Left, verticalAlignment=VAlign_Top}}
  }
  InputActionWidget {widgetClassPath=&CommonActionWidget, bInherited=true, widget=null}
 ]
}
